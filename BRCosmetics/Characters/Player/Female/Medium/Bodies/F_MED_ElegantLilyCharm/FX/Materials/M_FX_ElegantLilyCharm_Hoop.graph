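# Translucent unlit Niagara mesh-particle material for the hoop effect on the ElegantLilyCharm character cosmetic.
# Opacity layers noise-distorted, panning wave textures, shaped by vertex color, double-sided Fresnel and particle alpha, with depth fade skipped at low and medium quality.
# Emissive hue-shifts particle color over time, height and particle randomness, with noise variation.
# Vertex offset adds sine waves and normal-based thickness; dynamic parameters control thickness, tiling and UV offset.
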Material /BRCosmetics/Characters/Player/Female/Medium/Bodies/F_MED_ElegantLilyCharm/FX/Materials/M_FX_ElegantLilyCharm_Hoop
graph {
 properties {
  bAutomaticallySetUsageInEditor=false
  bUsedWithNiagaraMeshParticles=true
  blendMode=BLEND_Translucent
  shadingModel=MSM_Unlit
  usedShadingModels=Unlit
 }
 nodes [
  n1 = Normalize(VectorInput=Constant3Vector() {constant=<rgba 0 0 1 0>})
  n2 = ComponentMask(None=TextureCoordinate()) {r=true}
  n3 = Add(A=Multiply(A=Time()) {constB=0.125}, B=ParticleRandom())
  n4 = Multiply(A=ScalarParameter() {parameterName=Wave}, B=Sine(None=Add(A=@n2, B=@n3)))
  n5 = VertexColor()
  Add(
   A=Multiply(A=VertexNormalWS(), B=DynamicParameter() {paramNames [RevealWidth,Offset,Thickness,Tiling]}.Thickness)
   B=Multiply(A=Transform(None=@n1) {transformSourceType=TRANSFORMSOURCE_Instance}, B=Multiply(A=@n4, B=@n5.R))
  )
  n6 = NamedRerouteUsage()
  n7 = Subtract(A=Add(A=ParticleRandom(), B=Time()), B=Divide(A=WorldPosition().Z) {constB=250})
  n8 = Add(A=Multiply(A=Sine(None=@n7) {period=2}) {constB=0.5}) {constB=0.5}
  n9 = ScalarParameter() {parameterName=HueShift}
  n10 = ParticleColor()
  n11 = MaterialFunctionCall(Hue Shift Percentage=VertexInterpolator(VS=Multiply(A=@n8, B=@n9) {constB=0.5}).PS, Texture=@n10.RGB) {materialFunction=&HueShift}
  MaterialFunctionCall(Emissive=Add(A=Multiply(A=@n6, B=@n11.Result), B=@n11.Result)) {materialFunction=&MF_FX_EmissiveAdjustments}
  n12 = TextureCoordinate()
  n13 = DynamicParameter() {paramNames [RevealWidth,Offset,Thickness,Tiling]}
  n14 = Multiply(A=ComponentMask(None=@n12) {g=true}, B=@n13.Tiling)
  NamedRerouteDeclaration(None=MaterialFunctionCall(R=ComponentMask(None=@n12) {r=true}, G=Add(A=@n14, B=@n13.Offset)) {materialFunction=&AppendMany}.RG) {name=BaseTiledUV, nodeColor=<rgba 0.875 0 1 1>}
  n15 = NamedRerouteUsage()
  n16 = Constant() {r=0.7}
  n17 = TextureSample(UVs=Panner(Coordinate=Multiply(A=@n15, B=@n16)) {speedX=0.25, speedY=0.5}) {texture=/Game/Effects/Fort_Effects/Textures/Tile/T_Noise_03}
  n18 = ScalarParameter() {parameterName=distortion}
  n19 = MaterialFunctionCall(Float2=Add(A=Multiply(A=@n17.R, B=@n18), B=NamedRerouteUsage())) {materialFunction=&BreakOutFloat2Components}
  n20 = Panner(Coordinate=MaterialFunctionCall(R=@n19.R, G=Clamp(None=@n19.G)) {materialFunction=&AppendMany}.RG) {speedX=0.2}
  n21 = TextureSample(UVs=@n20) {texture=/BRCosmetics/Characters/Player/Female/Medium/Bodies/F_MED_ElegantLilyCharm/FX/T_FX_ElegantLilyCharm_Waves}
  n22 = Add(A=@n20, B=ComponentMask(None=VectorParameter() {parameterName=Layer2_Offset, defaultValue=<rgba 1 1 0 1>}.RGB) {g=true, r=true})
  n23 = NamedRerouteUsage()
  n24 = TextureSample(UVs=Panner(Coordinate=Multiply(A=@n23) {constB=0.35}) {speedX=0.25, speedY=0.2}) {texture=/Game/Effects/Fort_Effects/Textures/Power_Source/T_Power_Source_Sphere_Texture}
  n25 = Add(
   A=@n21.B
   B=Multiply(
    A=TextureSample(UVs=@n22) {texture=&T_FX_ElegantLilyCharm_Waves}.B
    B=Add(A=@n24.G) {constB=0.125}
   )
  )
  n26 = ScalarParameter() {parameterName=Glow}
  NamedRerouteDeclaration(None=Add(A=Saturate(None=@n25), B=Multiply(A=@n21.G, B=@n26))) {name=TextureOpacity, nodeColor=<rgba 1 0 0.7266 1>}
  n27 = Panner(Coordinate=TextureCoordinate() {uTiling=2, vTiling=4}) {speedX=0.5, speedY=-0.5}
  NamedRerouteDeclaration(None=Multiply(
    A=TextureSample(UVs=@n27) {texture=&T_Noise_03}.RGB
    B=ScalarParameter() {parameterName=chroma_noise}
   )) {name=ChromaNoise, nodeColor=<rgba 0 0.4375 1 1>}
  n28 = NamedRerouteUsage()
  n29 = VertexColor()
  n30 = Multiply(
   A=Multiply(A=@n10.A, B=Multiply(A=@n28, B=@n29.R))
   B=MaterialFunctionCall(Brightness=ScalarParameter() {defaultValue=1, parameterName=Fresnel_Brightness}, Exponent=ScalarParameter() {parameterName=Fresnel_Exponent}) {materialFunction=&MF_Fresnel_FX}.Fresnel Double Sided
  )
  QualitySwitch(Default=DepthFade(Opacity=@n30), Low=@n30, Medium=@n30)
 ]
}
