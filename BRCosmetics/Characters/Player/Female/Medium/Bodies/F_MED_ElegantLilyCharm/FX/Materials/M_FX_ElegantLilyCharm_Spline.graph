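# Translucent unlit Niagara mesh-particle material for the ElegantLilyCharm cosmetic's spline effect, with vertex red separating hoops from other strands.
# SplineThicken widens the mesh, and dynamic parameters drive thickness, UV tiling, offset and texture width.
# Noise-distorted, panning wave texture layers build opacity, with edge masking, a pulsing hoop reveal, camera fade and quality-switched depth fade.
# Color combines an animated four-color gradient, noise-driven hue shift and particle color.
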Material /BRCosmetics/Characters/Player/Female/Medium/Bodies/F_MED_ElegantLilyCharm/FX/Materials/M_FX_ElegantLilyCharm_Spline
graph {
 properties {
  bAutomaticallySetUsageInEditor=false
  bUsedWithNiagaraMeshParticles=true
  blendMode=BLEND_Translucent
  shadingModel=MSM_Unlit
  usedShadingModels=Unlit
 }
 nodes [
  Constant3Vector() {constant=<rgba 4 4 4 0>}
  Constant() {r=4}
  n1 = Multiply(
   A=DynamicParameter() {defaultValue=<rgba 1 1 2 1>, paramNames [TextureMaxWidth, Offset, Thickness, Tiling]}.Thickness
   B=LinearInterpolate(Alpha=NamedRerouteUsage()) {constA=1, constB=1.25}
  )
  MaterialFunctionCall(WidthBase=@n1, WidthTip=@n1, Expand U or V UV Channel=StaticBool() {value=true}) {materialFunction=&SplineThicken}
  n2 = TextureCoordinate()
  n3 = Add(A=Time(), B=ParticleRandom())
  n4 = Fmod(A=Add(A=ComponentMask(None=@n2) {g=true}, B=Multiply(A=@n3) {constB=0.5}), B=Constant() {r=1})
  n5 = Constant() {r=0.5}
  n6 = MaterialFunctionCall(
   1=VectorParameter() {group=Coloring, parameterName=ColorA, defaultValue=<rgba 0 0 0 1>}.RGB
   2=VectorParameter() {group=Coloring, parameterName=ColorB, defaultValue=<rgba 0 0 0 1>}.RGB
   3=VectorParameter() {group=Coloring, parameterName=ColorC, defaultValue=<rgba 0 0 0 1>}.RGB
   4=VectorParameter() {group=Coloring, parameterName=ColorD, defaultValue=<rgba 0 0 0 1>}.RGB
   A=Multiply(A=Distance(A=@n4, B=@n5)) {constB=2}
  ) {materialFunction=&Lerp_Multiple_Float3}
  n7 = Multiply(A=NamedRerouteUsage(), B=@n6.Lerp 4 Inputs)
  n8 = ParticleColor()
  MaterialFunctionCall(Emissive=Multiply(A=Add(A=@n7, B=@n6.Lerp 4 Inputs), B=@n8.RGB)) {materialFunction=&MF_FX_EmissiveAdjustments}
  MaterialFunctionCall(Hue Shift Percentage=NamedRerouteUsage(), Texture=VectorParameter() {parameterName=BaseColor, defaultValue=<rgba 1.315 0.391 2 1>}.RGB) {materialFunction=&HueShift}
  n9 = NamedRerouteUsage()
  n10 = SphereMask(A=ComponentMask(None=TextureCoordinate()) {g=true}, B=Constant() {r=0.5}) {attenuationRadius=0.5, hardnessPercent=0}
  n11 = Multiply(A=Multiply(A=@n8.A, B=Multiply(A=@n9, B=@n10)), B=NamedRerouteUsage())
  n12 = DepthFade(Opacity=@n11) {fadeDistanceDefault=3}
  n13 = TextureCoordinate()
  n14 = Time()
  n15 = ParticleRandom()
  n16 = Sine(None=Add(A=ComponentMask(None=@n13) {g=true}, B=Add(A=@n14, B=@n15))) {period=2}
  n17 = LinearInterpolate(B=VertexInterpolator(VS=Clamp(None=@n16)).PS, Alpha=NamedRerouteUsage()) {constA=1}
  Multiply(A=QualitySwitch(Default=@n12, Low=@n11, High=@n12, Medium=@n11, Epic=@n12), B=@n17)
  n18 = MaterialFunctionCall(Fade Length=Constant() {r=100}, Fade Offset=Constant() {r=128}, For Vertex Shader=StaticBool() {value=true}) {materialFunction=&CameraDepthFade}
  n19 = DynamicParameter() {parameterIndex=3, defaultValue=<rgba 1 1 2 0.5>, paramNames [TextureMaxWidth, Offset, Thickness, Tiling///HueShift]}
  NamedRerouteDeclaration(None=VertexInterpolator(VS=LinearInterpolate(A=@n18.Result, Alpha=@n19.TextureMaxWidth)).PS) {name=ADS, nodeColor=<rgba 0 0.1094 1 1>}
  n20 = TextureCoordinate() {vTiling=0.7}
  n21 = DynamicParameter() {defaultValue=<rgba 1 1 2 1>, paramNames [TextureMaxWidth, Offset, Thickness, Tiling]}
  n22 = ScalarParameter() {defaultValue=1, group=Hoops, parameterName=HoopLayer_Tiling}
  n23 = VertexColor()
  n24 = Multiply(A=ComponentMask(None=@n20) {g=true}, B=Multiply(A=@n21.Tiling, B=LinearInterpolate(A=@n22, Alpha=@n23.R) {constA=4}))
  NamedRerouteDeclaration(None=MaterialFunctionCall(R=ComponentMask(None=@n20) {r=true}, G=Add(A=@n24, B=@n21.Offset)) {materialFunction=&AppendMany}.RG) {name=BaseTiledUV, nodeColor=<rgba 0.875 0 1 1>}
  n25 = NamedRerouteUsage()
  n26 = VectorParameter() {parameterName=NoiseDistortion, defaultValue=<rgba 0 0 0 1>}
  n27 = Panner(Coordinate=Multiply(A=NamedRerouteUsage(), B=Constant() {r=0.7})) {speedX=0.25, speedY=0.5}
  n28 = Add(
   A=Multiply(A=TextureSample(UVs=@n27) {texture=/Game/Effects/Fort_Effects/Textures/Tile/T_Noise_03}.R, B=ScalarParameter() {parameterName=distortion})
   B=NamedRerouteUsage()
  )
  n29 = Constant() {r=1}
  n30 = ComponentMask(None=TextureCoordinate()) {g=true}
  n31 = Constant() {r=0.5}
  n32 = Multiply(A=Multiply(A=@n21.TextureMaxWidth, B=SphereMask(A=@n30, B=@n31) {attenuationRadius=0.5, hardnessPercent=0}), B=NamedRerouteUsage())
  n33 = MaterialFunctionCall(Float2=MaterialFunctionCall(
    UVs=MaterialFunctionCall(XY=@n28) {materialFunction=&Swizzle}.YX
    Texture Scale=AppendVector(A=@n29, B=@n32)
   ) {materialFunction=&ScaleUVsByCenter}.UVs) {materialFunction=&BreakOutFloat2Components}
  n34 = Clamp(None=@n33.G)
  n35 = Time()
  n36 = Multiply(A=@n35) {constB=1.5}
  n37 = NamedRerouteUsage()
  n38 = Add(
   A=Multiply(A=ComponentMask(None=@n25) {g=true, r=true}, B=ComponentMask(None=@n26.RGB) {g=true, r=true})
   B=Panner(
    Coordinate=MaterialFunctionCall(R=@n33.R, G=@n34) {materialFunction=&AppendMany}.RG
    Time=LinearInterpolate(A=@n35, B=@n36, Alpha=@n37)
   ) {bFractionalPart=true, speedX=0.2}
  )
  n39 = TextureSample(UVs=@n38) {texture=/BRCosmetics/Characters/Player/Female/Medium/Bodies/F_MED_ElegantLilyCharm/FX/T_FX_ElegantLilyCharm_Waves}
  n40 = VectorParameter() {parameterName=Layer2_Offset, defaultValue=<rgba 1 1 0 1>}
  n41 = TextureSample(UVs=Add(A=@n38, B=ComponentMask(None=@n40.RGB) {g=true, r=true})) {texture=&T_FX_ElegantLilyCharm_Waves}
  n42 = Panner(Coordinate=Multiply(A=NamedRerouteUsage()) {constB=0.35}) {speedX=0.25, speedY=0.2}
  n43 = Saturate(None=Add(A=TextureSample(UVs=@n42) {texture=/Game/Effects/Fort_Effects/Textures/Power_Source/T_Power_Source_Sphere_Texture}.G) {constB=0.75})
  n44 = Multiply(A=@n41.B, B=@n43)
  n45 = LinearInterpolate(A=Add(A=@n39.R, B=Multiply(A=@n41.R, B=@n43)), B=Multiply(A=Add(A=@n39.B, B=@n44)) {constB=0.125}, Alpha=NamedRerouteUsage()) {constAlpha=1}
  n46 = Multiply(A=ScalarParameter() {parameterName=Glow}, B=OneMinus(None=NamedRerouteUsage()))
  NamedRerouteDeclaration(None=Add(A=Saturate(None=@n45), B=Multiply(A=@n39.G, B=@n46))) {name=TextureOpacity, nodeColor=<rgba 1 0 0.7266 1>}
  n47 = Panner(Coordinate=TextureCoordinate() {uTiling=2, vTiling=4}) {speedY=1}
  NamedRerouteDeclaration(None=Multiply(
    A=TextureSample(UVs=@n47) {texture=&T_Noise_03}.RGB
    B=ScalarParameter() {parameterName=chroma_noise}
   )) {name=ChromaNoise, nodeColor=<rgba 0 0.4375 1 1>}
  n48 = Time()
  n49 = Multiply(A=NamedRerouteUsage()) {constB=0.5}
  n50 = TextureCoordinate()
  n51 = Panner(Coordinate=Multiply(A=NamedRerouteUsage(), B=Constant() {r=0.125}), Time=Add(A=Add(A=@n48, B=@n49), B=ComponentMask(None=@n50) {g=true})) {speedX=-1, speedY=0.1}
  n52 = DynamicParameter() {parameterIndex=3, defaultValue=<rgba 1 1 2 0.5>, paramNames [TextureMaxWidth, Offset, Thickness, Tiling///HueShift]}
  NamedRerouteDeclaration(None=Multiply(
    A=TextureSample(UVs=@n51) {texture=&T_Power_Source_Sphere_Texture}.G
    B=@n52.Tiling///HueShift
   ) {constB=0.5}) {name=HueShiftDriver, nodeColor=<rgba 0.7109 1 0 1>}
  NamedRerouteDeclaration(None=OneMinus(None=VertexColor().R)) {name=IsHoop, nodeColor=<rgba 0.9922 1 0 1>}
  NamedRerouteDeclaration(None=@n17) {name=HoopReveal, nodeColor=<rgba 0 1 0.7656 1>}
  ScalarParameter() {parameterName=Scale}
  VertexColor()
  n53 = Add(A=ParticleRandom(), B=Time())
  n54 = Divide(A=WorldPosition().Z) {constB=250}
  n55 = Multiply(A=Sine(None=Subtract(A=@n53, B=@n54)) {period=2}) {constB=0.5}
  VertexInterpolator(VS=Multiply(A=Add(A=@n55) {constB=0.5}, B=ScalarParameter() {parameterName=HueShift}) {constB=0.5})
 ]
}
