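# Idle cosmetic effect for the ElegantLilyCharm player body: glowing spline meshes and sparkles attached to the character skeleton.
# CharacterSplines places seven socket-bound mesh particles, each with its own mesh and material parameters.
# GPU Sparkles interpolate between sockets along two chains, while FlowerSparkles emit from a bone offset and hide with the head.
# System audio reactivity drives brightness, width and spawn rates, with time-of-day tints and user flags controlling visibility.
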NiagaraSystem /BRCosmetics/Characters/Player/Female/Medium/Bodies/F_MED_ElegantLilyCharm/FX/NS_ElegantLilyCharm_Idle
dependencies {
 usedDataInterfaces [
  &NiagaraDataInterfaceMeshRendererInfo
  &NiagaraDataInterfaceCurve
  &NiagaraDataInterfaceColorCurve
  &NiagaraDataInterfaceSkeletalMesh
 ]
 usedDynamicInputs [
  /Niagara/DynamicInputs/Lerp/SmoothLerpOverTime_Float
  &RemapRange
  &MakeFloatFromVector
  &ClampFloat
 ]
 usedModules [
  &self:SystemSpawnScript.SystemScriptSource.SystemScriptGraph.NiagaraNodeAssignment_0.SetVariables_#1
  &self:SystemSpawnScript.SystemScriptSource.SystemScriptGraph.NiagaraNodeAssignment_3.SetVariables_#2
  &self:SystemSpawnScript.SystemScriptSource.SystemScriptGraph.NiagaraNodeAssignment_9.SetVariables_#3
  &self:SystemSpawnScript.SystemScriptSource.SystemScriptGraph.NiagaraNodeAssignment_5.SetVariables_#4
  &SystemState
  &Timeline
  &SampleSkeletalMesh
  &self:SystemSpawnScript.SystemScriptSource.SystemScriptGraph.NiagaraNodeAssignment_8.SetVariables_#5
  &self:SystemSpawnScript.SystemScriptSource.SystemScriptGraph.NiagaraNodeAssignment_1.SetVariables_#6
  &self:SystemSpawnScript.SystemScriptSource.SystemScriptGraph.NiagaraNodeAssignment_6.SetVariables_#7
  &self:SystemSpawnScript.SystemScriptSource.SystemScriptGraph.NiagaraNodeAssignment_2.SetVariables_#8
  &self:SystemSpawnScript.SystemScriptSource.SystemScriptGraph.NiagaraNodeAssignment_7.SetVariables_#9
 ]
}
emitters {
 CharacterSplines {
  emitterSpawnScript [/SetVariables(
    Emitter.BaseSplineColor:LinearColor=<rgba 1.315 0.391 2 1>
    Emitter.BaseWidth:NiagaraFloat=&FN_FrontEndFloatModifier(Gameplay=10, Front End=8)
   ) {isSetParameters=true}]
  emitterUpdateScript [
   &EmitterState(Life Cycle Mode=::NewEnumerator1, Loop Behavior=::NewEnumerator1, Loop Duration Mode=::NewEnumerator1, UseLoopDelay=false)
   /SetVariables(Emitter.TODTint:LinearColor=&FN_Select_Color_from_Time_Of_Day(Morning=<rgba 1 1 1 0.8>, Day=<rgba 1 1 1 1>, Evening=<rgba 1 1 1 1>, Night=<rgba 0.3615 0.4008 1 1>)) {isSetParameters=true}
   &SpawnBurst_Instantaneous(Spawn Count=7, Use Spawn Probability=false, UseLoopCountLimit=false)
  ]
  particleSpawnScript [
   &PartitionParticles(
    ID to Partition:NiagaraInt32=&ReturnExecIndex()
    Number Of Partitions=7
    Particles Per Partition=1
    Partition Mode:ENiagara_PartitionMode=::NewEnumerator1
   )
   /SetVariables(Particles.SplineIndex:NiagaraInt32=$Particles.PartitionParticles.Partition) {isSetParameters=true}
   &InitializeParticle(
    Color=<rgba 2 2 2 2>
    Lifetime=$Emitter.CurrentLoopDuration
    Mesh Renderer Array Visibility Mode=::NewEnumerator1
    Mesh Renderer Info=DataInterface() {propertyValues {MeshRenderer=&self:CharacterSplines_0.NiagaraMeshRendererProperties_0}}
    Mesh Renderer Mesh Index:NiagaraInt32=$Particles.SplineIndex
    Sprite Size Mode=::NewEnumerator3
    UsePositionOffset=false
    Write Material Random=true
   )
   /SetVariables(Particles.MaterialRandom:NiagaraFloat=&Multiply_FloatByInt(Float=0.388, Integer=$Particles.SplineIndex)) {isSetParameters=true}
  ]
  particleUpdateScript [
   &ParticleState(Kill Particles When Lifetime Has Elapsed=false, Loop Particles Lifetime=true)
   /SetVariables(Particles.VisibilityTag=&MakeCustomIntFromBool(
     Bool=$User.bIsFrontEndPreview
     True INT=&MergeHelper_IntSwitch(0=0, 1=0, 2=1, 3=0, 4=0, 5=0, Index=$Particles.SplineIndex, Num=::NewEnumerator4)
     False INT=&MergeHelper_IntSwitch(
      0=&MakeCustomIntFromBool(Bool=$User.bisFootHidden, True INT=1, False INT=0)
      1=&MakeCustomIntFromBool(Bool=$User.bisFootHidden, True INT=1, False INT=0)
      2=0
      3=0
      4=0
      5=0
      Index=$Particles.SplineIndex
      Num=::NewEnumerator4
     )
    )) {isSetParameters=true}
   &SkeletalMeshLocation(
    Bone / Socket Index=$Particles.SplineIndex
    MaskSpawnGroup=false
    Mesh Sampling Type=::NewEnumerator5
    Skeletal Mesh=$System.Skeletal Mesh
    Socket Sampling Mode=::NewEnumerator5
   ) {module=SkeletalMeshLocation001}
   &SolveForcesAndVelocity(Clamp Velocity=false, Limit Acceleration=false)
   &DynamicMaterialParameters(
    Index 0 Param 1=&Multiply_Float(
     A=&Lerp_Float(A=0.85, B=1.5, Alpha=$System.AudioReactivity)
     B=&FN_FrontEndFloatModifier(Gameplay=1, Front End=1)
    )
    Index 0 Param 2=&Multiply_FloatByInt(Float=0.25, Integer=$Particles.SplineIndex)
    Index 0 Param 3=&Multiply_Float(
     A=&MergeHelper_FloatSwitch(
      0=$Emitter.BaseWidth
      1=$Emitter.BaseWidth
      2=$Emitter.BaseWidth
      3=$Emitter.BaseWidth
      4=$Emitter.BaseWidth
      5=$Emitter.BaseWidth
      6=3
      Index=$Particles.SplineIndex
      Num=::NewEnumerator5
     )
     B=&Lerp_Float(A=1, B=&FN_FrontEndFloatModifier(Gameplay=1.2, Front End=1.25), Alpha=$System.AudioReactivity)
    )
    Index 0 Param 4=&Multiply_Float(
     A=&MergeHelper_FloatSwitch(0=0.5, 1=0.6, 2=0.75, 3=1.1, 4=0.75, 5=1, 6=0.75, Index=$Particles.SplineIndex, Num=::NewEnumerator5)
     B=1
    )
    Index 3 Param 1=&FN_FrontEndFloatModifier(Gameplay=0, Front End=1)
    Index 3 Param 4=$System.BaseHueShift
    Module.Index0Param0Write=true
    Module.Index0Param1Write=true
    Module.Index0Param2Write=true
    Module.Index0Param3Write=true
    Module.Index3Param0Write=true
    Module.Index3Param1Write=true
    Module.Index3Param2Write=true
    Module.Index3Param3Write=true
    Write Parameter Index 3=true
   )
   &CameraOffset(Camera Offset Amount=&MergeHelper_FloatSwitch(0=4, 1=0, 2=0, 3=5, 4=4, 5=4, 6=3, Index=$Particles.SplineIndex, Num=::NewEnumerator5))
   /SetVariables(Particles.TODColor:LinearColor=&MultiplyLinearColor(
     Linear Color=$Particles.Initial.Color
     Scale Factor=&MakeVector4_FromLinearColor(LinearColor=$Emitter.TODTint)
    )) {isSetParameters=true}
   &ScaleColor(
    Color Value To Scale=$StackContext.TODColor
    Scale Alpha=&Multiply_Float(
     A=&FloatFromCurve(FloatCurve=DataInterface() {propertyValues {Curve {keys [<at 0 0.375>,<at 1 0.9>]}}}, CurveIndex=$System.AudioReactivity)
     B=&MakeCustomFloatFromBool(
      Bool=&SetBoolByIntComparison(A=$Particles.SplineIndex, B=1, Comparison Type=::NewEnumerator4)
      True Float=$System.CrouchWeight
      False Float=1
     )
    )
    Scale RGB=&VectorFromFloat(Value=&FloatFromCurve(FloatCurve=DataInterface() {propertyValues {Curve {keys [<at 0 0.85>,<at 0.8 0.925>,<at 1 1.5>]}}}, CurveIndex=$System.AudioReactivity))
    ScaleA=true
    ScaleRGB=true
   )
  ]
  rendererClasses [&NiagaraMeshRendererProperties]
  renderers [{rendererClass=&NiagaraMeshRendererProperties, rendererIndex=0}]
 }
 FlowerSparkles {
  simTarget=GPUComputeSim
  emitterUpdateScript [
   &EmitterState(
    Enable Distance Culling=true
    Enable Visibility Culling=true
    Life Cycle Mode=::NewEnumerator1
    Loop Duration=1
    MaxDistance=3000
    Module.bUseMaxDistance=true
    Module.bUseMinDistance=false
    Scalability Mode=::NewEnumerator1
    UseLoopDelay=false
   )
   &SpawnRate(
    SpawnRate=&FloatFromCurve(
     FloatCurve=DataInterface() {propertyValues {Curve {keys [<at 0.5 0.5 RCIM_Cubic>,<at 1 1.5 RCIM_Cubic>]}}}
     CurveIndex=$System.AudioReactivity
     Scale Curve=30
    )
    Use Spawn Probability=false
   )
  ]
  particleSpawnScript [
   &InitializeParticle(
    Color=&ColorFromCurve(
     ColorCurve=DataInterface() {propertyValues {
       AlphaCurve {keys [<at 0 0.5 RCIM_Cubic>]}
       BlueCurve {keys [<at 0 1 RCIM_Cubic>,<at 0.3059 0.2996>,<at 1 1 RCIM_Cubic>]}
       GreenCurve {keys [<at 0 0.4223 RCIM_Cubic>,<at 0.3059 0.0201>,<at 1 0.4496 RCIM_Cubic>]}
       RedCurve {keys [<at 0 0.958 RCIM_Cubic>,<at 0.3059 2>,<at 1 0.9952 RCIM_Cubic>]}
      }}
     CurveIndex=&RandomRangeFloat(Minimum=0, Maximum=1)
    )
    Lifetime=&RandomRangeFloat(Minimum=0.25, Maximum=0.5)
    Sprite Rotation Mode=::NewEnumerator2
    Sprite Size Mode=::NewEnumerator3
    Uniform Sprite Size=&RandomRangeFloat(Minimum=2, Maximum=5)
    UsePositionOffset=false
    Write Material Random=true
   )
   /SetVariables(Particles.SpawnGroup=$Emitter.SpawnGroup) {isSetParameters=true, module=SetVariables_#10}
   &SkeletalMeshLocation(
    Bone / Socket Index=3
    Bone Sampling Mode=::NewEnumerator4
    Interpolate Mesh Sampling=false
    MaskSpawnGroup=false
    Sampled Orientation Coordinate Space=::World
    Sampled Position Offset=<8 2.5 -8>
    Skeletal Mesh=$System.Skeletal Mesh
   )
   &ShapeLocation(
    Cylinder Height=1
    Cylinder Radius=4
    MaskSpawnGroup=false
    Override Seed=false
    Rotation Angle=0.35
    Rotation Axis=<1 0.2 0.25>
    Rotation Mode=::NewEnumerator1
    Shape Primitive=::NewEnumerator1
    Surface Only=false
    UseOwnerScale=true
   )
   &AddVelocity(
    Cone Angle=150
    Cone Axis=&RotateVectorByQuaternion(Vector To Rotate=<0 0 -1>, Quaternion=$Particles.SkeletalMeshLocation.SampledRotation)
    Inner Cone Angle=45
    Module.UseConeFalloff=false
    Module.UseVelDistribution=false
    Override Seed=false
    Rotation Coordinate Space=::World
    Velocity Mode=::NewEnumerator2
    Velocity Speed=&RandomRangeFloat(Minimum=1, Maximum=15)
   )
   &ApplyInitialForces()
   &AddVelocity(
    Rotation Coordinate Space=::World
    Velocity=&RotateVectorByQuaternion(Vector To Rotate=<0 0 -1>, Quaternion=$Particles.SkeletalMeshLocation.SampledRotation)
    Velocity Speed Scale=&RandomRangeFloat(Minimum=2, Maximum=20)
   ) {module=AddVelocity001}
  ]
  particleUpdateScript [
   &ParticleState()
   &SpriteRotationRate(Rotation Rate=&RandomRangeFloat(Minimum=-250, Maximum=250))
   &ScaleSpriteSize(Uniform Curve Sprite Scale=DataInterface() {propertyValues {Curve {keys [
        {interpMode=RCIM_Cubic, tangentMode=RCTM_User, time=0, value=0}
        {interpMode=RCIM_Cubic, tangentMode=RCTM_User, time=0.5, value=1}
        {tangentMode=RCTM_User, time=1, value=0}
       ]}}})
   &Drag(Drag=&RandomRangeFloat(Minimum=6, Maximum=10), Use Linear Drag=true, Use Rotational Drag=true)
   &InheritSourceMovement(
    Applied Position Scale=&VectorFromFloat(Value=&FloatFromCurve(FloatCurve=DataInterface() {propertyValues {Curve {keys [{interpMode=RCIM_Cubic, tangentMode=RCTM_User, time=0, value=1}, {interpMode=RCIM_Cubic, tangentMode=RCTM_User, time=1, value=0.9}]}}}))
    ApplyForce=false
    ApplyPosition=true
    ApplyPositionFalloff=false
   )
   &SolveForcesAndVelocity(Clamp Velocity=false, Limit Acceleration=false)
   /SetVariables(Particles.VisibilityTag=&MakeCustomIntFromBool(Bool=$User.bisHeadHidden, True INT=1, False INT=0)) {isSetParameters=true}
  ]
  rendererClasses [&NiagaraSpriteRendererProperties]
  renderers [
   {rendererClass=&NiagaraSpriteRendererProperties}
   {rendererClass=&NiagaraSpriteRendererProperties, rendererIndex=1}
  ]
 }
 Sparkles {
  simTarget=GPUComputeSim
  emitterSpawnScript [/SetVariables(Emitter.NumSparkleSockets=7, Emitter.NumSparkleSockets_Chain_2=5, Emitter.SparkleChain2_StartIndex=11, Emitter.StartingSparkleSocketIndex=6) {isSetParameters=true}]
  emitterUpdateScript [
   &EmitterState(
    Enable Distance Culling=true
    Enable Visibility Culling=true
    Life Cycle Mode=::NewEnumerator1
    Loop Duration=1
    MaxDistance=3000
    Module.bUseMaxDistance=true
    Module.bUseMinDistance=false
    Scalability Mode=::NewEnumerator1
    UseLoopDelay=false
   )
   /SetVariables(Emitter.AudioSpawnRate:NiagaraFloat=&FloatFromCurve(
     FloatCurve=DataInterface() {propertyValues {Curve {keys [<at 0.5 0.1>,<at 0.6 1>,<at 1 2>]}}}
     CurveIndex=$System.AudioReactivity
     Scale Curve=50
    )) {isSetParameters=true}
   &SpawnRate(SpawnGroup=1, SpawnRate=$Emitter.AudioSpawnRate, Use Spawn Probability=false) {module=SpawnRate001}
   &SpawnRate(SpawnRate=$Emitter.AudioSpawnRate, Use Spawn Probability=false)
  ]
  particleSpawnScript [
   &InitializeParticle(
    Color=&ColorFromCurve(
     ColorCurve=DataInterface() {propertyValues {
       AlphaCurve {keys [<at 0 1 RCIM_Cubic>]}
       BlueCurve {keys [<at 0 5 RCIM_Cubic>,<at 0.3059 0.749>,<at 0.5106 2>,<at 0.7853 1>,<at 1 1 RCIM_Cubic>]}
       GreenCurve {keys [<at 0 0.6 RCIM_Cubic>,<at 0.3059 0.0502>,<at 0.5106 0.391>,<at 0.7853 0.4644>,<at 1 1 RCIM_Cubic>]}
       RedCurve {keys [<at 0 4.1952 RCIM_Cubic>,<at 0.3059 5>,<at 0.5106 1.315>,<at 0.7853 0.0117>,<at 1 1 RCIM_Cubic>]}
      }}
     CurveIndex=&RandomRangeFloat(Minimum=0, Maximum=1)
    )
    Lifetime=&RandomRangeFloat(Minimum=0.5, Maximum=1)
    Sprite Rotation Mode=::NewEnumerator2
    Sprite Size Mode=::NewEnumerator3
    Uniform Sprite Size=&RandomRangeFloat(Minimum=2, Maximum=5)
    UsePositionOffset=false
    Write Material Random=true
   )
   /SetVariables(Particles.SpawnGroup=$Emitter.SpawnGroup) {isSetParameters=true, module=SetVariables_#10}
   /SetVariables(Particles.SampleT:NiagaraFloat=&RandomRangeFloat(Minimum=0, Maximum=1)) {isSetParameters=true}
   /SetVariables(Particles.SampleIndexFloat:NiagaraFloat=&Multiply_FloatByInt(
     Float=$Particles.SampleT
     Integer=&MergeHelper_IntSwitch(0=$Emitter.NumSparkleSockets, 1=$Emitter.NumSparkleSockets_Chain_2, 2=?, 3=?, 4=?, 5=?, Index=$Particles.SpawnGroup, Num=::NewEnumerator0)
    )) {isSetParameters=true}
   /SetVariables(Particles.SampleBlendWeight:NiagaraFloat=&ModuloFloat(A=&Multiply_FloatByInt(Float=$Particles.SampleT, Integer=$Emitter.NumSparkleSockets), B=1)) {isSetParameters=true}
   &SkeletalMeshLocation(
    Bone / Socket Index=&Add_Integer(
     A=&MergeHelper_IntSwitch(
      0=$Emitter.StartingSparkleSocketIndex
      1=$Emitter.SparkleChain2_StartIndex
      2=?
      3=?
      4=?
      5=?
      Index=$Particles.SpawnGroup
      Num=::NewEnumerator0
     )
     B=&MakeIntFromFloat(Float=$Particles.SampleIndexFloat, Conversion Mode=::NewEnumerator3)
    )
    MaskSpawnGroup=false
    Mesh Sampling Type=::NewEnumerator5
    Position Sampling=::NewEnumerator2
    Skeletal Mesh=$System.Skeletal Mesh
    Socket Sampling Mode=::NewEnumerator5
   )
   &SkeletalMeshLocation(
    Bone / Socket Index=&Add_Integer(
     A=&MergeHelper_IntSwitch(
      0=$Emitter.StartingSparkleSocketIndex
      1=$Emitter.SparkleChain2_StartIndex
      2=?
      3=?
      4=?
      5=?
      Index=$Particles.SpawnGroup
      Num=::NewEnumerator0
     )
     B=&MakeIntFromFloat(Float=$Particles.SampleIndexFloat, Conversion Mode=::NewEnumerator2)
    )
    MaskSpawnGroup=false
    Mesh Sampling Type=::NewEnumerator5
    Position Sampling=::NewEnumerator2
    Skeletal Mesh=$System.Skeletal Mesh
    Socket Sampling Mode=::NewEnumerator5
   ) {module=SkeletalMeshLocation001}
   /SetVariables(Particles.Position=&Lerp_Vector(A=$Particles.SkeletalMeshLocation.SampledPosition, B=$Particles.SkeletalMeshLocation001.SampledPosition, Alpha=$Particles.SampleBlendWeight)) {isSetParameters=true}
   &ShapeLocation(MaskSpawnGroup=false, Override Seed=false, Sphere Radius=7, UseOwnerScale=true)
   /SetVariables(Particles.SampledDirection:Vector3f=&NormalizeVector(
     Vector To Normalize=&SubtractVectorFromPosition(
      Position=Linked() {linkedVariable {name=StackContext.SkeletalMeshLocation001.SampledPosition, type=Vector3f}}
      Vector=Linked() {linkedVariable {name=StackContext.SkeletalMeshLocation.SampledPosition, type=NiagaraPosition}}
     )
     Safe Normalize=true
     Safe Fallback Threshold=0
     Safe Fallback Vector=<0 0 0.001>
    )) {isSetParameters=true}
   &AddVelocity(
    Cone Angle=1
    Cone Axis=$Particles.SampledDirection
    Module.UseConeFalloff=false
    Module.UseVelDistribution=false
    Override Seed=false
    Rotation Coordinate Space=::World
    Velocity Mode=::NewEnumerator2
    Velocity Speed=&RandomRangeFloat(Minimum=-10, Maximum=-25)
   )
   &ApplyInitialForces()
   /SetVariables(Particles.isLegSparkle:NiagaraBool=&SetBoolByIntComparison(A=&MakeIntFromFloat(Float=$Particles.SampleIndexFloat), B=4, Comparison Type=::NewEnumerator6)) {isSetParameters=true}
  ]
  particleUpdateScript [
   &ParticleState()
   &SpriteRotationRate(Rotation Rate=&RandomRangeFloat(Minimum=-250, Maximum=250))
   &ScaleSpriteSize(Uniform Curve Sprite Scale=DataInterface() {propertyValues {Curve {keys [
        {arriveTangent=9.0725, interpMode=RCIM_Cubic, leaveTangent=9.0725, tangentMode=RCTM_User, time=0, value=0}
        {arriveTangent=-2.5249, interpMode=RCIM_Cubic, leaveTangent=-2.5249, tangentMode=RCTM_User, time=0.1408, value=0.9847}
        {arriveTangent=-0.6532, interpMode=RCIM_Cubic, leaveTangent=-0.6532, time=0.3, value=0.375}
        {arriveTangent=-1.0221, interpMode=RCIM_Cubic, leaveTangent=-1.0221, tangentMode=RCTM_User, time=0.5, value=0.75}
        {arriveTangent=-0.0405, interpMode=RCIM_Cubic, leaveTangent=-0.0405, tangentMode=RCTM_User, time=1, value=0}
       ]}}})
   &Drag(Drag=&RandomRangeFloat(Minimum=1, Maximum=3), Use Linear Drag=true, Use Rotational Drag=true)
   &InheritSourceMovement(
    Applied Position Scale=&VectorFromFloat(Value=&FloatFromCurve(FloatCurve=DataInterface() {propertyValues {Curve {keys [{interpMode=RCIM_Cubic, tangentMode=RCTM_User, time=0, value=1}, {interpMode=RCIM_Cubic, tangentMode=RCTM_User, time=1, value=0.75}]}}}))
    ApplyForce=false
    ApplyPosition=true
    ApplyPositionFalloff=false
   )
   &SolveForcesAndVelocity(Clamp Velocity=false, Limit Acceleration=false)
   &ScaleColor(
    Scale Mode=::NewEnumerator1
    Scale RGBA=&MakeVector4_FromLinearColor(LinearColor=&MultiplyLinearColor(
      Linear Color=$Particles.Initial.Color
      Scale Factor=&MakeVector4_FromLinearColor(LinearColor=&FN_Select_Color_from_Time_Of_Day(Morning=<rgba 0.25 0.25 0.25 0.7>, Day=<rgba 0.5 0.5 0.5 0.8>, Evening=<rgba 1 1 1 1>, Night=<rgba 0.0567 0.3207 1 0.5>))
     ))
    ScaleRGBA=true
   )
   /SetVariables(
    Particles.GlowColor=&MultiplyLinearColor(
     Linear Color=$Particles.Color
     Scale Factor=&MakeFloat4FromVectorAndFloat(
      "Vector (RGB)"=&VectorFromFloat(Value=&Waveform(Global Phase=$Particles.Age, "Waveform [1]"=::NewEnumerator10, "[1] Frequency"=0.25, "[1] Amplitude Min/Max"=<0 1>))
      "Float (Alpha)"=0.52
     )
    )
    Particles.GlowSize=&Multiply_Vector2D_ByFloat(Float=2, Vector2D=$Particles.SpriteSize)
   ) {isSetParameters=true}
   /SetVariables(Particles.VisibilityTag=&MakeCustomIntFromBool(
     Bool=$Particles.isLegSparkle
     True INT=&MakeCustomIntFromBool(Bool=$User.bisFootHidden, True INT=1, False INT=0)
     False INT=0
    )) {isSetParameters=true}
  ]
  rendererClasses [&NiagaraSpriteRendererProperties]
  renderers [
   {rendererClass=&NiagaraSpriteRendererProperties}
   {rendererClass=&NiagaraSpriteRendererProperties, rendererIndex=1}
   {rendererClass=&NiagaraSpriteRendererProperties, rendererIndex=2}
  ]
 }
}
summary {
 systemName=NS_ElegantLilyCharm_Idle
 emitters [
  {emitterName=CharacterSplines, rendererClasses [&NiagaraMeshRendererProperties]}
  {emitterName=Sparkles, simTarget=GPUComputeSim, rendererClasses [&NiagaraSpriteRendererProperties]}
  {emitterName=FlowerSparkles, simTarget=GPUComputeSim, rendererClasses [&NiagaraSpriteRendererProperties]}
 ]
 userVariables [
  User.bisFootHidden {type=NiagaraBool}
  User.bIsFrontEnd {type=NiagaraBool}
  User.bIsFrontEndPreview {type=NiagaraBool}
  User.bisHeadHidden {type=NiagaraBool}
 ]
}
