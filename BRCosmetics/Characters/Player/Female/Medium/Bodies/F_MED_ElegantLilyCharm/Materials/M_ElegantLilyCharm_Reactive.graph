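# Audio-reactive character material for the ElegantLilyCharm outfit, built between the CharacterParent_2023 PreFX and PostFX functions.
# MF_AudioReactivity amplitude, with a debug override, provides the reactivity signal.
# One material-attribute blend adds animated idle/reactive color gradients with a Fresnel rim, masked to the dress; another adds flickering emissive glow.
# Static switches select sequins and reactivity style, alongside packed swirl SDF/gradient/ID textures and procedural pulse and sequin noise.
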Material /BRCosmetics/Characters/Player/Female/Medium/Bodies/F_MED_ElegantLilyCharm/Materials/M_ElegantLilyCharm_Reactive
graph {
 properties {
  bHasPixelAnimation=true
  bUseMaterialAttributes=true
  bUsedWithMeshDeformer=true
  bUsedWithMorphTargets=true
  bUsedWithSkeletalMesh=true
  pixelDepthOffsetMode=PDOM_Legacy
  previewMesh=/BRCosmetics/Characters/Player/Female/Medium/Bodies/F_MED_ElegantLilyCharm/Meshes/F_MED_ElegantLilyCharm
  shadingModel=MSM_FromMaterialExpression
  usedShadingModels=Default Lit | Subsurface | Clear Coat | Subsurface Profile | Cloth
 }
 nodes [
  n1 = NamedRerouteUsage()
  LinearInterpolate(Alpha=Saturate(None=@n1)) {constA=0.05, constAlpha=0, constB=0.1}
  n2 = NamedRerouteUsage()
  n3 = TextureSample(UVs=@n2) {desc=SequinUV Map, samplerType=SAMPLERTYPE_LinearColor}
  LinearInterpolate(Alpha=Ceil(None=@n3.R))
  MaterialFunctionCall(Speed=Constant() {r=2}) {materialFunction=&DebugTimeSine}
  n4 = AppendVector(A=@n3.R, B=@n3.R)
  n5 = Floor(None=@n2)
  n6 = TextureSample(UVs=Divide(A=Add(A=@n4, B=@n5), B=NamedRerouteUsage())) {
   samplerType=SAMPLERTYPE_LinearColor
   texture=/BRCosmetics/Characters/Player/Female/Medium/Bodies/F_MED_ElegantLilyCharm/Textures/T_ElegantLilyCharm_Body_FX2
  }
  Multiply(A=@n6.G)
  n7 = ScalarParameter() {defaultValue=2, group=MF_Sequin, parameterName=SequinSparkleSpeed}
  n8 = Divide(A=@n7) {constB=10}
  n9 = NamedRerouteUsage()
  n10 = TextureCoordinate()
  n11 = ScalarParameter() {defaultValue=32, group=MF_Sequin, parameterName=SequinTile}
  n12 = TextureSampleParameter2D(UVs=NamedRerouteDeclaration(None=Multiply(A=@n10, B=@n11)) {name=SequinUV, nodeColor=<rgba 0.1484 1 0 1>}) {
   group=MF_Sequin
   parameterName=Sequin_TilingOffsetPattern
   samplerType=SAMPLERTYPE_LinearGrayscale
   texture=/Game/Global/Textures/Default/Tileable_Patterns/T_SequinTile
  }
  n13 = Subtract(
   A=Multiply(A=Time(), B=LinearInterpolate(A=@n8, B=@n7, Alpha=@n9)) {constB=0.2}
   B=NamedRerouteDeclaration(None=@n12.R) {name=SequinIDs, nodeColor=<rgba 0 1 0.8594 1>}
  )
  NamedRerouteDeclaration(None=Frac(None=Sine(None=@n13))) {name=SequinNoise, nodeColor=<rgba 1 0.0343 0 1>}
  n14 = MaterialFunctionCall() {materialFunction=&MF_CharacterParent_2023_PreFX}
  NamedRerouteDeclaration(None=StaticSwitchParameter(
    True=MaterialFunctionCall!(Mask, TechArtMask=@n14.Tech Art Mask, inMat=@n14.Tech Art Mask) {materialFunction=/Game/Weapons/WeaponSkins/MaterialFunction/MF_Sequins}.Result
    False=@n14.Result
   ) {group=01_ElegantLilyCharm, parameterName=UseSequin, sortPriority=0}) {name=MA_AfterSequins, nodeColor=<rgba 0.5 1 0 1>}
  n15 = TextureSample() {
   samplerType=SAMPLERTYPE_LinearColor
   texture=&T_ElegantLilyCharm_Body_FX2
  }
  NamedRerouteDeclaration(None=AppendVector(A=@n15.B, B=@n15.B)) {name=SwirlGradient, nodeColor=<rgba 0 0.9766 1 1>}
  n16 = ComponentMask(None=NamedRerouteUsage()) {r=true}
  n17 = Time()
  n18 = NamedRerouteUsage()
  n19 = Add(
   A=Add(A=Subtract(A=@n17, B=@n18), B=ScalarParameter() {defaultValue=1, group=01_ElegantLilyCharm, parameterName=B_2})
   B=Multiply(A=NamedRerouteUsage(), B=ScalarParameter() {defaultValue=0.5, group=01_ElegantLilyCharm, parameterName=B_1})
  )
  n20 = Subtract(A=Divide(A=@n16) {constB=0.5}, B=@n19)
  n21 = Sine(None=Subtract(A=Multiply(A=@n19) {constB=1.2}, B=Divide(A=@n16) {constB=0.4})) {period=2}
  n22 = LinearInterpolate(A=Power(Base=Sine(None=@n20)) {constExponent=4}, Alpha=Multiply(A=Power(Base=@n21)))
  n23 = ComponentMask(None=NamedRerouteUsage()) {r=true}
  NamedRerouteDeclaration(None=Multiply(A=Abs(None=@n22), B=Ceil(None=@n23))) {name=Pulse, nodeColor=<rgba 1 0 0.8203 1>}
  NamedRerouteDeclaration(None=@n11) {name=SequinTiling, nodeColor=<rgba 1 0 0.4687 1>}
  NamedRerouteDeclaration(None=@n14.M) {name=MaskMap, nodeColor=<rgba 1 0.5625 0 1>}
  n24 = NamedRerouteUsage()
  n25 = Constant3Vector() {constant=<rgba 1 0.25 0.25 1>}
  n26 = Constant3Vector() {constant=<rgba 1 0.328 0.5408 1>}
  n27 = Constant3Vector() {constant=<rgba 1 0.5805 0.3672 1>}
  n28 = Multiply(A=Time()) {constB=0.3, desc=Default TIme Flow}
  n29 = NamedRerouteUsage()
  n30 = Add(A=Subtract(A=Add(A=@n28, B=@n29), B=NamedRerouteUsage()), B=NamedRerouteUsage())
  n31 = Add(A=Multiply(A=Sine(None=@n30)) {constB=0.5}) {constB=0.5}
  n32 = Constant3Vector() {constant=<rgba 1 0.0104 0.0104 1>}
  n33 = Constant3Vector() {constant=<rgba 1 0.2753 0.0949 1>}
  n34 = Constant3Vector() {constant=<rgba 1 0.4308 0.7587 1>}
  n35 = NamedRerouteUsage()
  n36 = LinearInterpolate(
   A=LinearInterpolate(
    A=MaterialFunctionCall(A=@n25, B=@n26, C=@n27, Alpha=@n31) {desc=default colors, materialFunction=&3ColorBlend}.Result
    B=MaterialFunctionCall(A=@n32, B=@n33, C=@n34, Alpha=@n31) {desc=reactive color, materialFunction=&3ColorBlend}.Result
    Alpha=NamedRerouteUsage()
   ) {constA=0.1, constAlpha=1, desc=Select Idle/Reactive Colors}
   B=Constant3Vector() {constant=<rgba 0 0.0864 5 1>}
   Alpha=Multiply(A=Fresnel() {baseReflectFraction=0, exponent=2}, B=LinearInterpolate(Alpha=@n35) {constA=0.5, constB=2}) {constB=2}
  )
  n37 = Multiply(A=LinearInterpolate(Alpha=NamedRerouteUsage()) {constA=0.01, constAlpha=1, constB=0.5}, B=@n36) {constB=5, desc=Define Emissive}
  n38 = NamedRerouteUsage()
  n39 = Multiply(A=Time()) {constB=0.3, desc=Default TIme Flow}
  n40 = Add(A=Multiply(A=Sine(None=@n39)) {constB=0.1}) {constB=0.4}
  n41 = NamedRerouteUsage()
  n42 = Multiply(A=@n41) {constB=0.2}
  n43 = Multiply(A=@n41) {constB=0.75}
  n44 = Multiply(A=Add(A=Subtract(A=@n38, B=@n40), B=LinearInterpolate(A=@n42, B=@n43, Alpha=@n1) {constAlpha=1}), B=NamedRerouteUsage())
  n45 = InverseLinearInterpolate(A=OneMinus(None=NamedRerouteUsage()), Value=Multiply(A=Saturate(None=@n44)) {constB=5})
  n46 = Max(A=NamedRerouteUsage()) {constB=0.05}
  n47 = Saturate(None=Multiply(A=Multiply(A=@n45, B=@n46)) {constB=4})
  n48 = LinearInterpolate(
   A=Step(X=NamedRerouteUsage()) {constX=0.5, constY=0.1}
   Alpha=ChannelMaskParameter(None=NamedRerouteUsage()) {maskChannel=Green, parameterName=DressMaskChannel, defaultValue=<rgba 0 1 0 0>}.R
  )
  NamedRerouteDeclaration(None=BlendMaterialAttributes(
    A=@n24
    B=SetMaterialAttributes!(MaterialAttributes=@n24, Emissive Color=@n37, Base Color=@n36) {attributeSetTypes [#1, #2]}
    Alpha=Multiply(A=@n47, B=@n48)
   )) {name=MA_AudioReactivity, nodeColor=<rgba 1 0 0.8203 1>}
  NamedRerouteDeclaration(None=@n15.G) {name=Swirl_ID, nodeColor=<rgba 0 0.9766 1 1>}
  NamedRerouteDeclaration(None=@n15.A) {name=SwirlSDF, nodeColor=<rgba 1 0.7031 0 1>}
  n49 = ScalarParameter() {group=01_ElegantLilyCharm, parameterName=Reactivity_Debug, sliderMax=1}
  n50 = LinearInterpolate(
   A=MaterialFunctionCall() {materialFunction=/Game/Packages/Fortress_SharedMaterials/Base_Material_Functions/BaseCharacterShaderFunctions/Audio/MF_AudioReactivity}.Amplitude
   B=@n49
   Alpha=@n49
  )
  n51 = ScalarParameter() {defaultValue=1, group=01_ElegantLilyCharm, parameterName=Reactivity}
  NamedRerouteDeclaration(None=Saturate(None=LinearInterpolate(B=@n50, Alpha=@n51))) {name=AudioReactivity_Amplitude, nodeColor=<rgba 0.3 0 1 1>}
  n52 = NamedRerouteUsage()
  n53 = VectorParameter() {group=01_ElegantLilyCharm, parameterName=Reactivity_EmissiveColor, defaultValue=<rgba 2 0.0991 0.3989 1>}
  n54 = Add(
   A=@n53.A
   B=MaterialFunctionCall!(Time Mult=Constant() {r=0.3}) {materialFunction=/Game/Effects/Fort_Effects/MaterialFunctions/MF_Flicker}.Processed Flicker
  )
  n55 = NamedRerouteUsage()
  n56 = NamedRerouteUsage()
  n57 = Multiply(A=@n53.RGB, B=Multiply(A=Multiply(A=@n54, B=@n55), B=ComponentMask(None=@n56) {g=true}))
  n58 = NamedRerouteUsage()
  NamedRerouteDeclaration(None=BlendMaterialAttributes(
    A=@n52
    B=SetMaterialAttributes!(MaterialAttributes=@n52, Emissive Color=@n57) {attributeSetTypes [#1]}
    Alpha=ChannelMaskParameter(None=@n58) {group=01_ElegantLilyCharm, parameterName=ReactivityMask_Channel, sortPriority=3}.R
   )) {name=MA_Glow, nodeColor=<rgba 0.9918 1 0.2351 1>}
  n59 = NamedRerouteUsage()
  n60 = NamedRerouteUsage()
  NamedRerouteDeclaration(None=StaticSwitchParameter(
    True=StaticSwitchParameter(True=@n59, False=@n60) {group=01_ElegantLilyCharm, parameterName=UseReactivityGlow, sortPriority=0}
    False=NamedRerouteUsage()
   ) {group=01_ElegantLilyCharm, parameterName=UseReactivity, sortPriority=0}) {name=MA_FInal, nodeColor=<rgba 0 1 0.2734 1>}
  n61 = NamedRerouteUsage()
  n62 = Constant2Vector() {g=0.25, r=-1}
  n63 = Add(A=Time(), B=NamedRerouteUsage())
  n64 = Add(
   A=Multiply(A=Multiply(A=@n61, B=@n62), B=ScalarParameter() {defaultValue=1.5, group=MF_Sequin, parameterName=SequinDiamondTile})
   B=Multiply(A=Multiply(A=@n63) {constB=0.05}, B=Constant2Vector() {r=1})
  )
  n65 = Constant() {r=0.15}
  NamedRerouteDeclaration(None=TextureSample(UVs=MaterialFunctionCall(UVs=@n64, "Rotation Angle (0-1)"=@n65) {materialFunction=&CustomRotator}.Rotated Values) {samplerType=SAMPLERTYPE_Masks, texture=/Game/Environments/TechArt/Hologram/T_GlitchNoise}.RGB) {name=ColorNoise, nodeColor=<rgba 0 1 0.8359 1>}
  NamedRerouteDeclaration(None=@n14.Tech Art Mask) {name=FXMask, nodeColor=<rgba 0 0.5078 1 1>}
  SetMaterialAttributes(
   MaterialAttributes=MaterialFunctionCall(Material=NamedRerouteUsage(), M=@n14.M, Pre Skinned Local Position=@n14.M, Distance Blend=@n14.M, Fresnel=@n14.M) {materialFunction=&MF_CharacterParent_2023_PostFX}.Result
   Emissive Color=NamedRerouteUsage()
  ) {attributeSetTypes [#1]}
 ]
}
