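# Post-style Anim Blueprint for the ElegantLilyCharm female medium body cosmetic, posing its skeleton from the attached parent mesh.
# Copy Pose From Mesh feeds the body's own Control Rig, whose result goes to the output pose.
# A separate chain of weighted multi-target constraints for the front and back skirt master bones is disconnected from the output.
# The event graph holds only empty event stubs.
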AnimBlueprint /BRCosmetics/Characters/Player/Female/Medium/Bodies/F_MED_ElegantLilyCharm/Meshes/F_MED_ElegantLilyCharm_AnimBP
target_skeleton=&/F_MED_ElegantLilyCharm_Skeleton
graphs [
 AnimGraph {nodes [
   n1 = "Constraint - skirt_back_master (PPPPPPPP)"!(
    ComponentPose=Local To Component()
    Alpha=1
    ConstraintWeights_0=0.5049
    ConstraintWeights_1:float=0.0349
    ConstraintWeights_2:float=0.0358
    ConstraintWeights_3:float=0.1071
    ConstraintWeights_4:float=0.1034
    ConstraintWeights_5:float=0.0001
    ConstraintWeights_6:float=0.1058
    ConstraintWeights_7:float=0.1081
   ) -> (Pose) {
    alphaBoolBlend {blendOption=Linear}
    boneToModify {boneName=skirt_back_master}
    constraintSetup [
     {targetBone {boneName=pelvis}}
     {targetBone {boneName=thigh_r}}
     {targetBone {boneName=thigh_l}}
     {targetBone {boneName=deform_glute_l}}
     {targetBone {boneName=deform_glute_r}}
     {targetBone {boneName=deform_groin_l}}
     {targetBone {boneName=skirt_thigh_twist_01_r}}
     {targetBone {boneName=skirt_thigh_twist_01_l}}
    ]
    constraintWeights [1,1,1,1,1,1,1,1]
   }
   n2 = "Constraint - skirt_front_master (PPPPPPPP)"!(
    ComponentPose=@n1
    Alpha=1
    ConstraintWeights_0=0.1911
    ConstraintWeights_1:float=0.0181
    ConstraintWeights_2:float=0.0182
    ConstraintWeights_3:float=0.0561
    ConstraintWeights_4:float=0.0561
    ConstraintWeights_5:float=0.2957
    ConstraintWeights_6:float=0.296
    ConstraintWeights_7:float=0.0688
   ) -> (Pose) {
    alphaBoolBlend {blendOption=Linear}
    boneToModify {boneName=skirt_front_master}
    constraintSetup [
     {targetBone {boneName=pelvis}}
     {targetBone {boneName=thigh_r}}
     {targetBone {boneName=thigh_l}}
     {targetBone {boneName=deform_groin_l}}
     {targetBone {boneName=deform_groin_r}}
     {targetBone {boneName=skirt_thigh_twist_01_r}}
     {targetBone {boneName=skirt_thigh_twist_01_l}}
     {targetBone {boneName=skirt_front}}
    ]
    constraintWeights [1,1,1,1,1,1,1,1]
   }
   Component To Local(ComponentPose=@n2)
   n3 = Control Rig(Source=Copy Pose From Mesh() {bUseAttachedParent=true}) {
    alphaBoolBlend {blendOption=Linear}
    controlRigAssetReference {blueprintRigClass=&/F_MED_ElegantLilyCharm_CtrlRig.F_MED_ElegantLilyCharm_CtrlRig_C}
   }
   Output Pose(Result=Component To Local(ComponentPose=Local To Component(LocalPose=@n3))) {layerGroup=None}
  ]}
 EventGraph {nodes [
   Event BlueprintUpdateAnimation()
   Event BlueprintInitializeAnimation()
   Event OnCharacterCustomizationLoaded()
  ]}
]
variables [As Player Pawn Athena {type=PlayerPawn_Athena_C}]
