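# Physics asset for the back hair part of the ElegantLilyCharm medium female player character, defining bodies and joints for hair motion.
# A main chain runs from a base sphere through three capsules to a box tip, with two curl-corrective bodies and an extra segment branching off.
# Joints lock selected swing or twist axes to restrict movement.
# Unconstrained spine, clavicle and upper-arm capsules provide torso and shoulder collision shapes.
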PhysicsAsset /BRCosmetics/Characters/Player/Female/Medium/Bodies/F_MED_ElegantLilyCharm/Meshes/Parts/F_MED_ElegantLilyCharm_FaceAcc_BackHairBack_Physics
bodies {
 Base_hair_back_L [{shapeType=Sphere, radius=1.0734, shapeName=Base_hair_back_L_sphere}]
 Dyn_hair_back_L_01 [{
   shapeType=Capsule
   length=13.9281
   radius=4.3254
   shapeName=Dyn_hair_back_L_01_capsule
   center=<3.8171 -0.2245 1.5527>
   rotation=<rot 0 0 90>
  }]
 Dyn_hair_back_L_02 [{
   shapeType=Capsule
   length=25.0437
   radius=4.5157
   shapeName=Dyn_hair_back_L_02_capsule
   center=<4.4424 -0.3155 1.5329>
   rotation=<rot 0 0 90>
  }]
 Dyn_hair_back_L_03 [{shapeType=Capsule, length=20.263, radius=3.4846, shapeName=Dyn_hair_back_L_03_capsule, center=<3.9423 -0.2823 0.0775>, rotation=<rot 0 0 90>}]
 Dyn_hair_back_L_04 [{
   shapeType=Box
   extentX=14.6288
   extentY=22.1413
   extentZ=7.3812
   shapeName=Dyn_hair_back_L_04_box
   center=<6.7612 -0.4547 1.0646>
   rotation=<rot 6.5553 0 0>
  }]
 Dyn_hair_back_L_06 [{
   shapeType=Capsule
   length=6.3368
   radius=2.0464
   shapeName=Dyn_hair_back_L_06_capsule
   center=<3.3548 0.5007 0.2007>
   rotation=<rot 86.6058 -175.3954 165.3163>
  }]
 clavicle_l [{
   shapeType=Capsule
   length=4.6779
   radius=6.1717
   shapeName=clavicle_l_capsule
   center=<6.3931 -1.0251 -2.2485>
   rotation=<rot 86.9643 -0.0014 -0.0017>
  }]
 clavicle_r [{
   shapeType=Capsule
   length=4.6779
   radius=6.1717
   shapeName=clavicle_l_capsule
   center=<-6.3931 1.0251 2.2485>
   rotation=<rot 86.9643 -0.0014 179.9983>
  }]
 dyn_hairCurlCorr_l [{
   shapeType=Capsule
   length=2.4547
   radius=1.6587
   shapeName=dyn_hairCurlCorr_l_capsule
   center=<2.0332 0.65 -0.2096>
   rotation=<rot 81.0273 7.3311 -13.4567>
  }]
 dyn_hairCurlCorr_r [{
   shapeType=Capsule
   length=1.762
   radius=2.3514
   shapeName=dyn_hairCurlCorr_l_capsule
   center=<1.513 -0.0615 -0.1138>
   rotation=<rot 81.0271 7.3317 15.3417>
  }]
 spine_01 [{shapeType=Capsule, length=5.9018, radius=10.3682, shapeName=spine_01_capsule, center=<0 -0.0314 0>}]
 spine_02 [{shapeType=Capsule, length=4.0085, radius=9.4215, shapeName=spine_02_capsule, center=<0 1.395 0>}]
 spine_03 [{shapeType=Capsule, length=4.8535, radius=9.5365, shapeName=spine_03_capsule, center=<-2.0838 2.2176 0>}]
 spine_04 [{shapeType=Capsule, length=5.2699, radius=9.8801, shapeName=spine_04_capsule, center=<0 2.6192 0>}]
 spine_05 [{shapeType=Capsule, length=10.2422, radius=8.8577, shapeName=spine_05_capsule, center=<0.0464 0.9404 0>}]
 upperarm_l [{
   shapeType=Capsule
   length=24.7198
   radius=5.2802
   shapeName=upperarm_l_capsule
   center=<12.6605 -0.1968 -0.0207>
   rotation=<rot -89.9062 179.9462 -179.9461>
  }]
 upperarm_r [{
   shapeType=Capsule
   length=24.7198
   radius=5.2802
   shapeName=upperarm_l_capsule
   center=<-12.6605 0.1968 0.0207>
   rotation=<rot -89.9062 179.9462 0.0539>
  }]
}
constraints [
 {
  bone1Name=Dyn_hair_back_L_01
  bone2Name=Base_hair_back_L
  swing1LimitDegrees=22
  swing1Motion=Locked
  swing2LimitDegrees=50
  twistLimitDegrees=10
 }
 {
  bone1Name=Dyn_hair_back_L_04
  bone2Name=Dyn_hair_back_L_03
  swing1LimitDegrees=25
  swing2LimitDegrees=25
  twistLimitDegrees=10
  twistMotion=Locked
 }
 {bone1Name=Dyn_hair_back_L_02, bone2Name=Dyn_hair_back_L_01, swing1LimitDegrees=5, swing2LimitDegrees=10, twistLimitDegrees=15}
 {
  bone1Name=Dyn_hair_back_L_03
  bone2Name=Dyn_hair_back_L_02
  swing1LimitDegrees=20
  swing2LimitDegrees=30
  swing2Motion=Locked
  twistLimitDegrees=10
  twistMotion=Locked
 }
 {
  bone1Name=dyn_hairCurlCorr_r
  bone2Name=Dyn_hair_back_L_01
  swing1LimitDegrees=45
  swing1Motion=Locked
  swing2LimitDegrees=60
  twistLimitDegrees=45
  twistMotion=Locked
 }
 {
  bone1Name=dyn_hairCurlCorr_l
  bone2Name=Dyn_hair_back_L_01
  swing1LimitDegrees=45
  swing1Motion=Locked
  swing2LimitDegrees=45
  twistLimitDegrees=45
  twistMotion=Locked
 }
 {
  bone1Name=Dyn_hair_back_L_06
  bone2Name=Dyn_hair_back_L_02
  swing1LimitDegrees=45
  swing1Motion=Locked
  swing2LimitDegrees=0
  swing2Motion=Locked
  twistLimitDegrees=45
  twistMotion=Locked
 }
]
