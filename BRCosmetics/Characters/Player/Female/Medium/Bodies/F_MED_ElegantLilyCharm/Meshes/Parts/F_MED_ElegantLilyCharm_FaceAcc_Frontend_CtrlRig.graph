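# Control Rig for the ElegantLilyCharm female character's face accessory that poses the front hair bones in the frontend.
# The isFrontend bool weights a chain of averaged parent constraints binding left and right hair bones to spine, clavicle and arm bones, with transform offsets after each side.
# A separate branch, not linked to Begin Execution, scales a Base_hair_Front_L corrective offset by remapped left-clavicle delta pitch times isFrontend.
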ControlRigBlueprint /BRCosmetics/Characters/Player/Female/Medium/Bodies/F_MED_ElegantLilyCharm/Meshes/Parts/F_MED_ElegantLilyCharm_FaceAcc_Frontend_CtrlRig
graphs [
 RigVMFunctionLibrary {links [], nodes []}
 RigVMModel {nodes [
   Get Delta Ref = RigVMFunctionReferenceNode(ExecuteContext=(), Item=(Type=Bone,Name="clavicle_l"))
   Cast = &RigVMFunction_MathBoolToFloat(Value=RigVMVariableNode() {variable=isFrontend}.Value)
   Remap = &RigVMFunction_MathDoubleRemap(SourceMaximum=20, SourceMinimum=0, TargetMaximum=1, TargetMinimum=0, Value=@"Get Delta Ref".Result.Rotation.Pitch)
   Multiply = &RigVMFunction_MathDoubleMul(B=@Cast.Result, A=@Remap.Result)
   > &RigUnit_ModifyTransforms(
    ItemToModify=((Item=(Type=Bone,Name="Base_hair_Front_L"),Transform=(Rotation=(X=-0.0868,Y=0.4924,Z=0.1504,W=0.8529),Translation=(X=0,Y=0,Z=7),Scale3D=(X=1,Y=1,Z=1))))
    WeightMinimum=0
    Weight=@Multiply.Result
   )
   RigUnit_BeginExecution = &RigUnit_BeginExecution()
   n1 = &RigVMFunction_MathBoolToFloat(Value=RigVMVariableNode() {variable=isFrontend}.Value)
   &RigUnit_ParentConstraint(
    AdvancedSettings=(InterpolationType=Average,RotationOrderForFilter=XZY)
    Child=(Type=Bone,Name="Base_hair_Front_L")
    Parents="((Item=(Type=Bone,Name=\"upperarm_twist_01_l\"),Weight=0.0195),(Item=(Type=Bone,Name=\"spine_04\"),Weight=0.0479),(Item=(Type=Bone,Name=\"spine_05\"),Weight=0.4686),(Item=(Type=Bone,Name=\"clavicle_l\"),Weight=0.4638),(Item=(Type=Bone,Name=\"clavicle_r\"),Weight=0), ...(truncated)"
    ExecutePin=@RigUnit_BeginExecution.ExecutePin
    Weight=@n1.Result
   )
   > &RigUnit_ParentConstraint(
    AdvancedSettings=(InterpolationType=Average,RotationOrderForFilter=XZY)
    Child=(Type=Bone,Name="Dyn_hair_Front_L_02")
    Parents="((Item=(Type=Bone,Name=\"upperarm_twist_01_l\"),Weight=0.0463),(Item=(Type=Bone,Name=\"spine_04\"),Weight=0.3772),(Item=(Type=Bone,Name=\"spine_05\"),Weight=0.3752),(Item=(Type=Bone,Name=\"clavicle_l\"),Weight=0.2013),(Item=(Type=Bone,Name=\"clavicle_r\"),Weight=0), ...(truncated)"
    Weight=@n1.Result
   )
   > &RigUnit_ParentConstraint(
    AdvancedSettings=(InterpolationType=Average,RotationOrderForFilter=XZY)
    Child=(Type=Bone,Name="Dyn_hair_Front_L_03")
    Parents="((Item=(Type=Bone,Name=\"upperarm_twist_01_l\"),Weight=0.0113),(Item=(Type=Bone,Name=\"spine_04\"),Weight=0.7242),(Item=(Type=Bone,Name=\"spine_05\"),Weight=0.2357),(Item=(Type=Bone,Name=\"clavicle_l\"),Weight=0.0287),(Item=(Type=Bone,Name=\"upperarm_l\"),Weight=0), ...(truncated)"
    Weight=@n1.Result
   )
   > &RigUnit_ParentConstraint(
    AdvancedSettings=(InterpolationType=Average,RotationOrderForFilter=XZY)
    Child=(Type=Bone,Name="Dyn_hair_Front_L_04")
    Parents="((Item=(Type=Bone,Name=\"upperarm_twist_01_l\"),Weight=0.0113),(Item=(Type=Bone,Name=\"spine_04\"),Weight=0.7242),(Item=(Type=Bone,Name=\"spine_05\"),Weight=0.2357),(Item=(Type=Bone,Name=\"clavicle_l\"),Weight=0.0287),(Item=(Type=Bone,Name=\"upperarm_l\"),Weight=0), ...(truncated)"
    Weight=@n1.Result
   )
   > &RigUnit_ModifyTransforms(
    ItemToModify="((Item=(Type=Bone,Name=\"Dyn_hair_Front_L_01\"),Transform=(Rotation=(X=-0.0634,Y=0.0209,Z=0.0902,W=0.9937),Translation=(X=0.0004,Y=-0.0004,Z=3.0179),Scale3D=(X=1,Y=1,Z=1))),(Item=(Type=Bone,Name=\"Dyn_hair_Front_L_02\"),Transform=(Rotation=(X=0.0604,Y=-0.188,Z ...(truncated)"
    WeightMinimum=0
    Weight=@n1.Result
   )
   > &RigUnit_ParentConstraint(
    AdvancedSettings=(InterpolationType=Average,RotationOrderForFilter=XZY)
    Child=(Type=Bone,Name="Base_hair_Front_R")
    Parents="((Item=(Type=Bone,Name=\"upperarm_twist_01_r\"),Weight=0.0083),(Item=(Type=Bone,Name=\"spine_04\"),Weight=0.06),(Item=(Type=Bone,Name=\"spine_05\"),Weight=0.7745),(Item=(Type=Bone,Name=\"clavicle_l\"),Weight=0),(Item=(Type=Bone,Name=\"clavicle_r\"),Weight=0.157),(It ...(truncated)"
    Weight=@n1.Result
   )
   > &RigUnit_ParentConstraint(
    AdvancedSettings=(InterpolationType=Average,RotationOrderForFilter=XZY)
    Child=(Type=Bone,Name="Dyn_hair_Front_R_02")
    Parents="((Item=(Type=Bone,Name=\"upperarm_twist_01_r\"),Weight=0.0463),(Item=(Type=Bone,Name=\"spine_04\"),Weight=0.3772),(Item=(Type=Bone,Name=\"spine_05\"),Weight=0.3752),(Item=(Type=Bone,Name=\"clavicle_l\"),Weight=0),(Item=(Type=Bone,Name=\"clavicle_r\"),Weight=0.2013), ...(truncated)"
    Weight=@n1.Result
   )
   > &RigUnit_ParentConstraint(
    AdvancedSettings=(InterpolationType=Average,RotationOrderForFilter=XZY)
    Child=(Type=Bone,Name="Dyn_hair_Front_R_03")
    Parents="((Item=(Type=Bone,Name=\"upperarm_twist_01_l\"),Weight=0.0088),(Item=(Type=Bone,Name=\"spine_04\"),Weight=0.842),(Item=(Type=Bone,Name=\"spine_05\"),Weight=0.1286),(Item=(Type=Bone,Name=\"clavicle_r\"),Weight=0.0205),(Item=(Type=Bone,Name=\"upperarm_r\"),Weight=0),( ...(truncated)"
    Weight=@n1.Result
   )
   > &RigUnit_ParentConstraint(
    AdvancedSettings=(InterpolationType=Average,RotationOrderForFilter=XZY)
    Child=(Type=Bone,Name="Dyn_hair_Front_R_04")
    Parents=((Item=(Type=Bone,Name="lowerarm_twist_02_r"),Weight=0),(Item=(Type=Bone,Name="spine_04"),Weight=0.9884),(Item=(Type=Bone,Name="spine_05"),Weight=0.0116))
    Weight=@n1.Result
   )
   > &RigUnit_ModifyTransforms(
    ItemToModify="((Item=(Type=Bone,Name=\"Dyn_hair_Front_R_01\"),Transform=(Rotation=(X=0,Y=-0.0358,Z=0,W=-0.9994),Translation=(X=0,Y=0,Z=0),Scale3D=(X=1,Y=1,Z=1))),(Item=(Type=Bone,Name=\"Dyn_hair_Front_R_02\"),Transform=(Rotation=(X=0,Y=0.0254,Z=0,W=-0.9997),Translation=(X=0 ...(truncated)"
    WeightMinimum=0
    Weight=@n1.Result
   )
  ]}
]
hierarchy [
 root {loc=<0 0 0>, rot=<0 0 0 1>}
 pelvis {parent=root, loc=<0 -0.5301 98.0571>, rot=<0.0127 0.707 -0.0127 -0.707>}
 spine_01 {parent=pelvis, loc=<6.1208 -0.9328 0>, rot=<0 0 0.034 -0.9994>}
 spine_02 {parent=spine_01, loc=<6.5996 -0.2095 0>, rot=<0 0 0.0238 -0.9997>}
 spine_03 {parent=spine_02, loc=<6.4914 -0.1237 0>, rot=<0 0 0.0149 -0.9999>}
 spine_04 {parent=spine_03, loc=<6.9598 0.0794 0>, rot=<0 0 -0.0132 -0.9999>}
 spine_05 {parent=spine_04, loc=<6.9987 0 0>, rot=<0 0 -0.0015 -1>}
 attach_cape {parent=spine_05, loc=<9.3216 -8.4501 0>, rot=<0.0616 -0.7044 -0.0616 -0.7044>}
 clavicle_l {parent=spine_05, loc=<5.4831 -0.1053 -3.2946>, rot=<0.0465 -0.7289 -0.009 -0.6829>}
 upperarm_l {parent=clavicle_l, loc=<9.8658 -1.2136 -0.4105>, rot=<0.0164 -0.3245 -0.0269 -0.9454>}
 lowerarm_l {parent=upperarm_l, loc=<26.465 -0.8345 0.259>, rot=<0.0036 0.0161 -0.2298 -0.9731>}
 hand_l {parent=lowerarm_l, loc=<22.7355 -0.3203 -0.6105>, rot=<-0.6784 -0.0223 0.0735 -0.7307>}
 weapon_l {parent=hand_l, loc=<-0.1974 -3.4111 0.3811>, rot=<0 0 0 1>}
 thumb_01_l {parent=hand_l, loc=<1.3118 -0.3191 -2.459>, rot=<0.6546 -0.4042 0.0664 -0.6354>}
 thumb_02_l {parent=thumb_01_l, loc=<5.9336 -0.0448 0.0401>, rot=<-0.0037 0.0213 0.298 -0.9543>}
 thumb_03_l {parent=thumb_02_l, loc=<2.2223 0.3682 -0.0907>, rot=<0.0118 -0.013 0.0066 -0.9998>}
 index_metacarpal_l {parent=hand_l, loc=<1.6535 0.0183 -2.96>, rot=<0.0158 -0.0124 -0.0173 -0.9996>}
 index_01_l {parent=index_metacarpal_l, loc=<8.3008 0.0357 -0.0524>, rot=<-0.0044 0.0485 0.0924 -0.9945>}
 index_02_l {parent=index_01_l, loc=<3.3587 -0.3004 0.0452>, rot=<-0.002 0.0027 0.3225 -0.9466>}
 index_03_l {parent=index_02_l, loc=<2.5578 -0.123 -0.0252>, rot=<-0.002 0.0067 0.2558 -0.9667>}
 middle_metacarpal_l {parent=hand_l, loc=<2.1856 -0.1725 -0.5984>, rot=<0.006 0.0168 -0.0329 -0.9993>}
 middle_01_l {parent=middle_metacarpal_l, loc=<7.3275 -0.2447 0.0129>, rot=<-0.0208 0.0042 0.0858 -0.9961>}
 middle_02_l {parent=middle_01_l, loc=<4.0123 0.1401 0.1175>, rot=<-0.0218 0.0258 0.3512 -0.9357>}
 middle_03_l {parent=middle_02_l, loc=<2.6077 0.0534 0.1373>, rot=<0.0082 0.0188 0.1708 -0.9851>}
 ring_metacarpal_l {parent=hand_l, loc=<2.2074 -0.3625 1.3638>, rot=<-0.0284 0.0722 -0.0458 -0.9959>}
 ring_01_l {parent=ring_metacarpal_l, loc=<7.0295 -0.6197 -0.1446>, rot=<-0.0489 -0.0252 0.1892 -0.9804>}
 ring_02_l {parent=ring_01_l, loc=<3.6174 0.4647 -0.1027>, rot=<0 0 0.2806 -0.9598>}
 ring_03_l {parent=ring_02_l, loc=<2.4152 0.114 0.0199>, rot=<0 0 0.1189 -0.9929>}
 pinky_metacarpal_l {parent=hand_l, loc=<1.7616 -0.9753 3.3814>, rot=<-0.0397 0.0814 -0.069 -0.9935>}
 pinky_01_l {parent=pinky_metacarpal_l, loc=<7.1709 -0.4831 -0.1094>, rot=<-0.0654 -0.0035 0.2242 -0.9723>}
 pinky_02_l {parent=pinky_01_l, loc=<2.8274 0.0893 -0.1043>, rot=<0.0037 0.0011 0.2854 -0.9584>}
 pinky_03_l {parent=pinky_02_l, loc=<2.2046 0.0307 -0.0078>, rot=<-0.0002 0.0098 0.1445 -0.9895>}
 hand_attach_l {parent=hand_l, loc=<8.5 -3 0>, rot=<0.7071 0.7071 0 0>}
 deform_wrist_dn_base_l {parent=hand_l, loc=<0 -1.5326 0>, rot=<0 0 0 -1>}
 deform_wrist_dn_l {parent=deform_wrist_dn_base_l, loc=<0 0 0>, rot=<0 0 0 1>}
 deform_wrist_up_base_l {parent=hand_l, loc=<0 2.3009 0>, rot=<0 0 0 -1>}
 deform_wrist_up_l {parent=deform_wrist_up_base_l, loc=<0 0 0>, rot=<0 0 0 1>}
 lowerarm_twist_02_l {parent=lowerarm_l, loc=<10.6707 0 0>, rot=<0 0 0 1>}
 lowerarm_twist_01_l {parent=lowerarm_l, loc=<18.316 0 0>, rot=<0 0 0 1>}
 deform_elbow_in_base_l {parent=lowerarm_l, loc=<0.1834 3.6062 -0.2572>, rot=<0 0 0 -1>}
 deform_elbow_in_l {parent=deform_elbow_in_base_l, loc=<0 0 0>, rot=<0 0 0 1>}
 deform_elbow_out_base_l {parent=lowerarm_l, loc=<0 -4 0>, rot=<0 0 0 1>}
 deform_elbow_out_l {parent=deform_elbow_out_base_l, loc=<0 0 0>, rot=<0 0 0 1>}
 upperarm_twist_01_l {parent=upperarm_l, loc=<7.2135 -0.1134 0>, rot=<0 0 0 1>}
 deform_upperarm_bk_base_l {parent=upperarm_twist_01_l, loc=<-6.4053 -4.3746 0.314>, rot=<0 0 0 -1>}
 deform_upperarm_bk_l {parent=deform_upperarm_bk_base_l, loc=<0 0 0>, rot=<0 0 0 1>}
 deform_upperarm_fr_base_l {parent=upperarm_twist_01_l, loc=<-5.6312 5.092 0.3354>, rot=<0 0 0 -1>}
 deform_upperarm_fr_l {parent=deform_upperarm_fr_base_l, loc=<0 0 0>, rot=<0 0 0 1>}
 deform_upperarm_up_base_l {parent=upperarm_twist_01_l, loc=<-8.891 0.4825 3.9517>, rot=<0 0 0 -1>}
 deform_upperarm_up_l {parent=deform_upperarm_up_base_l, loc=<0 0 0>, rot=<0 0 0 1>}
 upperarm_twist_02_l {parent=upperarm_l, loc=<19.7998 -0.1134 0>, rot=<0 0 0 1>}
 hairClavicleCorr_l {parent=clavicle_l, loc=<6.0994 -1.2432 4.5671>, rot=<-0.0012 0.0335 0.0362 -0.9988>}
 attach_backpack {parent=spine_05, loc=<0 -8.35 0>, rot=<-0.0093 -0.707 0.0093 -0.707>}
 neck_01 {parent=spine_05, loc=<11.3138 0.1529 0>, rot=<0 0 -0.2003 -0.9797>}
 neck_02 {parent=neck_01, loc=<4.5627 0 0>, rot=<0 0 0.0504 -0.9987>}
 head {parent=neck_02, loc=<5.3914 0 0>, rot=<0 0 0.115 -0.9934>}
 flower_main {parent=head, loc=<7.7058 1.8174 -7.1757>, rot=<-0.1151 -0.5536 -0.0854 -0.8204>}
 flower_low {parent=flower_main, loc=<0 0 0>, rot=<0 0 0 1>}
 flower_high {parent=flower_low, loc=<0 0 0>, rot=<0 0 0 1>}
 leaves_03 {parent=flower_low, loc=<0.7641 -0.123 2.5001>, rot=<-0.4237 0.5357 -0.4828 -0.5481>}
 leaves_01 {parent=flower_low, loc=<0.73 -2.1358 -2.9023>, rot=<0.552 -0.6396 -0.4106 -0.343>}
 leaves_02 {parent=flower_low, loc=<0.2941 -3.2019 0.4838>, rot=<0.6093 -0.7681 0.0679 0.185>}
 Dyn_curl_L_01 {parent=head, loc=<4.014 4.7316 -5.9741>, rot=<0.3912 -0.8838 -0.2432 -0.082>}
 Dyn_curl_L_02 {parent=Dyn_curl_L_01, loc=<2.3695 0.3756 0.4219>, rot=<-0.1478 0.2134 -0.5498 -0.794>}
 Dyn_curl_R_01 {parent=head, loc=<10.5684 8.2276 0.0924>, rot=<0.5945 0.2685 0.755 0.0666>}
 Dyn_curl_R_02 {parent=Dyn_curl_R_01, loc=<2.2899 -0.0931 -1.1628>, rot=<-0.1487 -0.4717 0.2633 -0.8283>}
 Dyn_curl_R_03 {parent=Dyn_curl_R_02, loc=<2.5975 -0.0363 -0.1496>, rot=<-0.0111 0.0439 -0.3897 -0.9198>}
 hairCorrBack {parent=head, loc=<-0.0479 -6.939 0>, rot=<0.1819 -0.6833 0.6833 -0.1819>}
 dyn_earRing {parent=head, loc=<-0.1906 -0.2814 -6.1119>, rot=<0.0041 -0.9874 0.1583 -0.0007>}
 hairSideCorrCheek_r {parent=head, loc=<-3.0398 4.6549 7.163>, rot=<0.0029 -0.7071 -0.0029 -0.7071>}
 clavicle_r {parent=spine_05, loc=<5.4831 -0.1053 3.2946>, rot=<0.7289 0.0465 -0.6829 0.009>}
 upperarm_r {parent=clavicle_r, loc=<-9.8658 1.2136 0.4105>, rot=<0.0164 -0.3245 -0.0269 -0.9454>}
 lowerarm_r {parent=upperarm_r, loc=<-26.465 0.8345 -0.259>, rot=<0.0036 0.0161 -0.2298 -0.9731>}
 hand_r {parent=lowerarm_r, loc=<-22.7355 0.3203 0.6106>, rot=<-0.6784 -0.0223 0.0735 -0.7307>}
 hand_attach_r {parent=hand_r, loc=<-8.5001 2.9999 0>, rot=<0 0 -0.7071 -0.7071>}
 weapon_r {parent=hand_r, loc=<0.1974 3.4111 -0.3811>, rot=<0 0 0 1>}
 thumb_01_r {parent=hand_r, loc=<-1.3118 0.3191 2.459>, rot=<0.6546 -0.4042 0.0664 -0.6354>}
 thumb_02_r {parent=thumb_01_r, loc=<-5.9336 0.0448 -0.0401>, rot=<-0.0037 0.0213 0.298 -0.9543>}
 thumb_03_r {parent=thumb_02_r, loc=<-2.2223 -0.3682 0.0907>, rot=<0.0118 -0.013 0.0066 -0.9998>}
 index_metacarpal_r {parent=hand_r, loc=<-1.6535 -0.0183 2.96>, rot=<0.0158 -0.0124 -0.0173 -0.9996>}
 index_01_r {parent=index_metacarpal_r, loc=<-8.3008 -0.0357 0.0524>, rot=<-0.0044 0.0485 0.0924 -0.9945>}
 index_02_r {parent=index_01_r, loc=<-3.3587 0.3004 -0.0452>, rot=<-0.002 0.0027 0.3225 -0.9466>}
 index_03_r {parent=index_02_r, loc=<-2.5578 0.123 0.0252>, rot=<-0.002 0.0067 0.2558 -0.9667>}
 middle_metacarpal_r {parent=hand_r, loc=<-2.1856 0.1725 0.5984>, rot=<0.006 0.0168 -0.0329 -0.9993>}
 middle_01_r {parent=middle_metacarpal_r, loc=<-7.3275 0.2447 -0.0129>, rot=<-0.0208 0.0042 0.0858 -0.9961>}
 middle_02_r {parent=middle_01_r, loc=<-4.0123 -0.1401 -0.1175>, rot=<-0.0218 0.0258 0.3512 -0.9357>}
 middle_03_r {parent=middle_02_r, loc=<-2.6077 -0.0534 -0.1373>, rot=<0.0082 0.0188 0.1708 -0.9851>}
 ring_metacarpal_r {parent=hand_r, loc=<-2.2074 0.3625 -1.3638>, rot=<-0.0284 0.0722 -0.0458 -0.9959>}
 ring_01_r {parent=ring_metacarpal_r, loc=<-7.0295 0.6197 0.1446>, rot=<-0.0489 -0.0252 0.1892 -0.9804>}
 ring_02_r {parent=ring_01_r, loc=<-3.6174 -0.4647 0.1027>, rot=<0 0 0.2806 -0.9598>}
 ring_03_r {parent=ring_02_r, loc=<-2.4152 -0.114 -0.0199>, rot=<0 0 0.1189 -0.9929>}
 pinky_metacarpal_r {parent=hand_r, loc=<-1.7616 0.9753 -3.3814>, rot=<-0.0397 0.0814 -0.069 -0.9935>}
 pinky_01_r {parent=pinky_metacarpal_r, loc=<-7.1709 0.4831 0.1094>, rot=<-0.0654 -0.0035 0.2242 -0.9723>}
 pinky_02_r {parent=pinky_01_r, loc=<-2.8274 -0.0893 0.1043>, rot=<0.0037 0.0011 0.2854 -0.9584>}
 pinky_03_r {parent=pinky_02_r, loc=<-2.2046 -0.0307 0.0078>, rot=<-0.0002 0.0098 0.1445 -0.9895>}
 deform_wrist_dn_base_r {parent=hand_r, loc=<-0.0001 1.5328 0>, rot=<0 0 0 -1>}
 deform_wrist_dn_r {parent=deform_wrist_dn_base_r, loc=<0 0 0>, rot=<0 0 0 1>}
 deform_wrist_up_base_r {parent=hand_r, loc=<-0.0002 -2.3007 -0.0001>, rot=<0 0 0 -1>}
 deform_wrist_up_r {parent=deform_wrist_up_base_r, loc=<0 0 0>, rot=<0 0 0 1>}
 lowerarm_twist_02_r {parent=lowerarm_r, loc=<-10.6707 0 0>, rot=<0 0 0 -1>}
 lowerarm_twist_01_r {parent=lowerarm_r, loc=<-18.316 0 0>, rot=<0 0 0 -1>}
 deform_elbow_in_base_r {parent=lowerarm_r, loc=<-0.1834 -3.6062 0.2572>, rot=<0 0 0 -1>}
 deform_elbow_in_r {parent=deform_elbow_in_base_r, loc=<0 0 0>, rot=<0 0 0 1>}
 deform_elbow_out_base_r {parent=lowerarm_r, loc=<0 4 -0.0001>, rot=<0 0 0 -1>}
 deform_elbow_out_r {parent=deform_elbow_out_base_r, loc=<0 0 0>, rot=<0 0 0 1>}
 upperarm_twist_01_r {parent=upperarm_r, loc=<-7.2135 0.1134 0>, rot=<0 0 0 -1>}
 deform_upperarm_bk_base_r {parent=upperarm_twist_01_r, loc=<6.4053 4.3746 -0.314>, rot=<0 0 0 -1>}
 deform_upperarm_bk_r {parent=deform_upperarm_bk_base_r, loc=<0 0 0>, rot=<0 0 0 1>}
 deform_upperarm_fr_base_r {parent=upperarm_twist_01_r, loc=<5.6312 -5.092 -0.3354>, rot=<0 0 0 -1>}
 deform_upperarm_fr_r {parent=deform_upperarm_fr_base_r, loc=<0 0 0>, rot=<0 0 0 1>}
 deform_upperarm_up_base_r {parent=upperarm_twist_01_r, loc=<8.891 -0.4825 -3.9517>, rot=<0 0 0 -1>}
 deform_upperarm_up_r {parent=deform_upperarm_up_base_r, loc=<0 0 0>, rot=<0 0 0 1>}
 upperarm_twist_02_r {parent=upperarm_r, loc=<-19.7998 0.1134 0>, rot=<0 0 0 -1>}
 hairClavicleCorr_r {parent=clavicle_r, loc=<-6.9233 0.8691 -4.9186>, rot=<0.9976 -0.0668 0.0195 -0.0016>}
 Base_hair_back_L {parent=spine_05, loc=<12.9411 -6.6943 0>, rot=<0.2064 -0.6763 0.6763 -0.2064>}
 Dyn_hair_back_L_01 {parent=Base_hair_back_L, loc=<0 0 0>, rot=<0 0 0 1>}
 Dyn_hair_back_L_02 {parent=Dyn_hair_back_L_01, loc=<7.2475 0 0>, rot=<0 -0.2705 0 -0.9627>}
 Dyn_hair_back_L_03 {parent=Dyn_hair_back_L_02, loc=<7.42 0 0>, rot=<0 0.0028 0 -1>}
 Dyn_hair_back_L_04 {parent=Dyn_hair_back_L_03, loc=<7.1032 0 0>, rot=<0 -0.0305 0 -0.9995>}
 Dyn_hair_back_L_12 {parent=Dyn_hair_back_L_04, loc=<1.4918 -9.6525 -1.4633>, rot=<0 -0.2283 0 -0.9736>}
 Dyn_hair_back_L_14 {parent=Dyn_hair_back_L_12, loc=<5.9275 1.1029 0.3188>, rot=<-0.1175 0.4067 0.0157 -0.9058>}
 Dyn_hair_back_L_13 {parent=Dyn_hair_back_L_04, loc=<0.7925 -10.4317 0.8789>, rot=<-0.0193 -0.1454 0.2869 -0.9467>}
 Dyn_hair_back_L_07 {parent=Dyn_hair_back_L_04, loc=<8.4829 5.6888 -0.8046>, rot=<0.1915 0.2118 -0.1717 -0.9429>}
 Dyn_hair_back_L_08 {parent=Dyn_hair_back_L_04, loc=<9.4681 3.3934 -0.0625>, rot=<0.0572 0.1711 -0.0106 -0.9835>}
 Dyn_hair_back_L_09 {parent=Dyn_hair_back_L_04, loc=<9.8781 0.7144 1.482>, rot=<-0.0141 0.0145 0.1036 -0.9944>}
 Dyn_hair_back_L_10 {parent=Dyn_hair_back_L_04, loc=<8.3999 -1.9085 -1.2085>, rot=<-0.1258 0.1645 0.0425 -0.9774>}
 Dyn_hair_back_L_11 {parent=Dyn_hair_back_L_04, loc=<6.8319 -4.8132 -0.0603>, rot=<-0.0107 -0.2012 -0.0518 -0.9781>}
 Dyn_hair_back_L_06 {parent=Dyn_hair_back_L_02, loc=<7.6332 -12.2877 -2.1852>, rot=<-0.0053 -0.0234 0.1806 -0.9833>}
 dyn_hairCurlCorr_l {parent=Dyn_hair_back_L_01, loc=<4.0507 8.9393 -1.4297>, rot=<0.0964 -0.194 -0.5243 -0.8235>}
 dyn_hairCurlCorr_r {parent=Dyn_hair_back_L_01, loc=<3.649 -12.4012 -2.9168>, rot=<-0.0928 -0.1914 0.3655 -0.9062>}
 Base_hair_Front_R {parent=spine_05, loc=<9.6689 5.1325 7.6689>, rot=<0.1943 0.6248 0.7403 0.1547>}
 Dyn_hair_Front_R_01 {parent=Base_hair_Front_R, loc=<0 0 0>, rot=<0 0 0 1>}
 Dyn_hair_Front_R_02 {parent=Dyn_hair_Front_R_01, loc=<6.9076 0 0>, rot=<0.0942 0.0033 -0.0622 -0.9936>}
 Dyn_hair_Front_R_03 {parent=Dyn_hair_Front_R_02, loc=<6.4541 0.3801 0.0004>, rot=<0.0044 -0.0886 0.232 -0.9687>}
 Dyn_hair_Front_R_04 {parent=Dyn_hair_Front_R_03, loc=<5.4086 0 0>, rot=<0.0107 -0.2203 -0.2055 -0.9535>}
 Dyn_hair_Front_R_05 {parent=Dyn_hair_Front_R_03, loc=<4.5961 -1.4979 0.3246>, rot=<0.0713 -0.1 0.0303 -0.992>}
 Base_hair_Front_L {parent=spine_05, loc=<10.5877 2.5481 -7.8042>, rot=<-0.2049 -0.6854 -0.6655 -0.2129>}
 Dyn_hair_Front_L_01 {parent=Base_hair_Front_L, loc=<0 0 0>, rot=<-0.0003 0.0028 0.0007 -1>}
 Dyn_hair_Front_L_02 {parent=Dyn_hair_Front_L_01, loc=<7.0277 0 0>, rot=<-0.0497 0.0371 0.0437 -0.9971>}
 Dyn_hair_Front_L_03 {parent=Dyn_hair_Front_L_02, loc=<6.6657 0 0>, rot=<-0.0136 -0.1554 0.0143 -0.9877>}
 Dyn_hair_Front_L_04 {parent=Dyn_hair_Front_L_03, loc=<5.7002 0 0>, rot=<0.0354 0.0022 -0.0835 -0.9959>}
 Dyn_hair_Front_L_05 {parent=Dyn_hair_Front_L_03, loc=<0.1084 1.3078 0.0273>, rot=<-0.005 0.1243 -0.275 -0.9534>}
 thigh_l {parent=pelvis, loc=<-1.0855 -0.2109 -8.7053>, rot=<0.0009 0.0355 0.0155 -0.9993>}
 calf_l {parent=thigh_l, loc=<-42.5284 -0.0019 -0.7297>, rot=<0.0006 -0.0161 -0.0391 -0.9991>}
 foot_l {parent=calf_l, loc=<-46.7401 -2.5014 -0.592>, rot=<0.0427 -0.0238 0.0353 -0.9982>}
 ball_l {parent=foot_l, loc=<-3.5658 10.8066 -0.1544>, rot=<0.0001 -0.0001 -0.7063 -0.7079>}
 calf_twist_02_l {parent=calf_l, loc=<-19.1546 -1.7943 -0.2638>, rot=<0.0237 0.0095 -0.0035 -0.9997>}
 calf_twist_01_l {parent=calf_l, loc=<-35.8483 -1.9197 -0.5794>, rot=<0.0237 0.0095 -0.0035 -0.9997>}
 foot_corr_bk_l {parent=calf_twist_01_l, loc=<-6.8935 -4.7046 -0.0364>, rot=<-0.0496 -0.727 0.0133 -0.6847>}
 foot_pos_corr_l {parent=calf_l, loc=<-43.778 -2.0574 -0.7969>, rot=<0.0427 -0.0238 0.0353 -0.9982>}
 ball_pos_corr_l {parent=foot_pos_corr_l, loc=<-8.4492 9.5049 -0.1128>, rot=<0.0001 -0.0001 -0.7063 -0.7079>}
 deform_knee_in_base_l {parent=calf_l, loc=<0.8031 -5.0231 0.2456>, rot=<0 0 0 -1>}
 deform_knee_in_l {parent=deform_knee_in_base_l, loc=<0 0 0>, rot=<0 0 0 1>}
 deform_knee_out_base_l {parent=calf_l, loc=<0 4 0>, rot=<0 0 0 -1>}
 deform_knee_out_l {parent=deform_knee_out_base_l, loc=<0 0 0>, rot=<0 0 0 1>}
 thigh_twist_01_l {parent=thigh_l, loc=<-27.952 0 0>, rot=<0.0013 0 0 -1>}
 deform_groin_base_l {parent=thigh_l, loc=<0.4801 5.6776 -0.2486>, rot=<0 0 0 -1>}
 deform_groin_l {parent=deform_groin_base_l, loc=<0 0 0>, rot=<0 0 0 1>}
 skirt_thigh_twist_01_l {parent=thigh_l, loc=<-18.1685 0 0>, rot=<0.0013 0 0 -1>}
 thigh_r {parent=pelvis, loc=<-1.0855 -0.2109 8.7053>, rot=<0.0355 -0.0009 0.9993 0.0155>}
 calf_r {parent=thigh_r, loc=<42.5284 0.0019 0.7297>, rot=<0.0006 -0.0161 -0.0391 -0.9991>}
 foot_r {parent=calf_r, loc=<46.7401 2.5014 0.592>, rot=<0.0427 -0.0238 0.0353 -0.9982>}
 ball_r {parent=foot_r, loc=<3.5658 -10.8066 0.1544>, rot=<0.0001 -0.0001 -0.7063 -0.7079>}
 calf_twist_02_r {parent=calf_r, loc=<19.1546 1.7943 0.2638>, rot=<0.0237 0.0095 -0.0035 -0.9997>}
 calf_twist_01_r {parent=calf_r, loc=<35.8483 1.9197 0.5794>, rot=<0.0237 0.0095 -0.0035 -0.9997>}
 foot_corr_bk_r {parent=calf_twist_01_r, loc=<6.8935 4.7046 0.0364>, rot=<-0.0496 -0.727 0.0133 -0.6847>}
 foot_pos_corr_r {parent=calf_r, loc=<43.778 2.0574 0.7968>, rot=<0.0427 -0.0238 0.0353 -0.9982>}
 ball_pos_corr_r {parent=foot_pos_corr_r, loc=<8.4491 -9.5049 0.1128>, rot=<0.0001 -0.0001 -0.7063 -0.7079>}
 deform_knee_in_base_r {parent=calf_r, loc=<-0.8031 5.0231 -0.2457>, rot=<0 0 0 -1>}
 deform_knee_in_r {parent=deform_knee_in_base_r, loc=<0 0 0>, rot=<0 0 0 1>}
 deform_knee_out_base_r {parent=calf_r, loc=<0 -4 0>, rot=<0 0 0 -1>}
 deform_knee_out_r {parent=deform_knee_out_base_r, loc=<0 0 0>, rot=<0 0 0 1>}
 thigh_twist_01_r {parent=thigh_r, loc=<27.952 0 0>, rot=<0.0013 0 0 -1>}
 deform_groin_base_r {parent=thigh_r, loc=<-0.4801 -5.6776 0.2486>, rot=<0 0 0 -1>}
 deform_groin_r {parent=deform_groin_base_r, loc=<0 0 0>, rot=<0 0 0 1>}
 skirt_thigh_twist_01_r {parent=thigh_r, loc=<17.6455 0 0>, rot=<0.0013 0 0 -1>}
 deform_glute_base_l {parent=pelvis, loc=<-7.3466 -9.0213 -9.1245>, rot=<0.0172 0.031 -0.4477 -0.8935>}
 deform_glute_l {parent=deform_glute_base_l, loc=<0 0 0>, rot=<0 0 0 1>}
 deform_glute_base_r {parent=pelvis, loc=<-7.3464 -9.0213 9.1245>, rot=<0.031 -0.0172 0.8935 -0.4477>}
 deform_glute_r {parent=deform_glute_base_r, loc=<0 0 0>, rot=<0 0 0 1>}
 skirt_back_master {parent=pelvis, loc=<-8.8995 -14.4873 -0.0746>, rot=<-0.1016 -0.9948 -0.0004 -0.0065>}
 skirt_back {parent=skirt_back_master, loc=<0 0 0>, rot=<0 0 0 1>}
 skirt_front_master {parent=pelvis, loc=<-20.4446 9.3253 -0.0031>, rot=<0 0 0 1>}
 skirt_front {parent=skirt_front_master, loc=<0 0 0>, rot=<0 0 0 1>}
 Skirt_corrective {parent=pelvis, loc=<-16.8993 -12.4262 1.4>, rot=<-0.1071 -0.9943 0 0>}
 attach {parent=root, loc=<0 0 0>, rot=<0 0 0 1>}
 ik_hand_root {parent=root, loc=<0 0 0>, rot=<0 0 0 1>}
 ik_hand_gun {parent=ik_hand_root, loc=<-48.1743 3.3545 104.4381>, rot=<0.6276 0.1309 0.4183 -0.6434>}
 ik_hand_r {parent=ik_hand_gun, loc=<0 0 0>, rot=<0 0 0 1>}
 ik_hand_l {parent=ik_hand_gun, loc=<59.3271 67.6997 34.3524>, rot=<-0.6158 -0.7027 -0.3565 0>}
 ik_foot_root {parent=root, loc=<0 0 0>, rot=<0 0 0 1>}
 ik_foot_l {parent=ik_foot_root, loc=<14.8262 -7.3545 8.296>, rot=<0.0357 0.7045 0.0284 -0.7082>}
 ik_foot_r {parent=ik_foot_root, loc=<-14.8262 -7.3545 8.296>, rot=<0.7082 0.0284 -0.7045 0.0357>}
 root_fp {parent=root, loc=<0 0 0>, rot=<0 0 0 1>}
 attach_fp {parent=root_fp, loc=<0 0 0>, rot=<0 0 0 1>}
 camera_root {parent=root_fp, loc=<0 0 0>, rot=<0 0 0 1>}
 camera {parent=camera_root, loc=<0 0 0>, rot=<0 0 0 1>}
 attach_emote_01 {parent=root, loc=<0 0 0>, rot=<0 0 0 1>}
 attach_emote_02 {parent=root, loc=<0 0 0>, rot=<0 0 0 1>}
 attach_emote_03 {parent=root, loc=<0 0 0>, rot=<0 0 0 1>}
 warp_point_01 {parent=root, loc=<0 0 0>, rot=<0 0 0 1>}
 warp_point_02 {parent=root, loc=<0 0 0>, rot=<0 0 0 1>}
 warp_point_03 {parent=root, loc=<0 0 0>, rot=<0 0 0 1>}
]
variables [isFrontend {type=bool, default=False}]
