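# Physics asset for the face accessory mesh of the female medium ElegantLilyCharm player cosmetic, simulating left and right front hair strands.
# Each strand chains four capsule bodies from a small base sphere through constraints with locked twist and limited swing, generally loosening toward the tip on the second swing axis.
# Unconstrained capsules on the upper spine, clavicles and upper arms provide body shapes for the hair to collide with.
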PhysicsAsset /BRCosmetics/Characters/Player/Female/Medium/Bodies/F_MED_ElegantLilyCharm/Meshes/Parts/F_MED_ElegantLilyCharm_FaceAcc_Physics
bodies {
 Base_hair_Front_L [{shapeType=Sphere, radius=1.6809, shapeName=Base_hair_Front_L_sphere}]
 Base_hair_Front_R [{shapeType=Sphere, radius=1.0079, shapeName=Base_hair_Front_R_sphere}]
 Dyn_hair_Front_L_01 [{
   shapeType=Capsule
   length=4.7939
   radius=2.8647
   shapeName=Dyn_hair_Front_L_01_capsule
   center=<3.8966 0.2806 1.5241>
   rotation=<rot -81.9208 85.8604 4.3563>
  }]
 Dyn_hair_Front_L_02 [{
   shapeType=Capsule
   length=4.3501
   radius=2.4217
   shapeName=Dyn_hair_Front_L_02_capsule
   center=<3.4865 0 1.0852>
   rotation=<rot -76.3002 179.9925 -179.99>
  }]
 Dyn_hair_Front_L_03 [{
   shapeType=Capsule
   length=4.1451
   radius=2.1489
   shapeName=Dyn_hair_Front_L_03_capsule
   center=<3.5057 0.0001 0.8145>
   rotation=<rot 88.8689 -179.9943 -179.9956>
  }]
 Dyn_hair_Front_L_04 [{
   shapeType=Capsule
   length=6.7382
   radius=1.9878
   shapeName=Dyn_hair_Front_L_04_capsule
   center=<4.6612 0 -0.4659>
   rotation=<rot 75.2337 0.0001 0.0001>
  }]
 Dyn_hair_Front_R_01 [{
   shapeType=Capsule
   length=3.5808
   radius=3.5385
   shapeName=Dyn_hair_Front_R_01_capsule
   center=<3.3389 -0.3958 1.3174>
   rotation=<rot 83.6184 108.1377 -165.7837>
  }]
 Dyn_hair_Front_R_02 [{
   shapeType=Capsule
   length=3.1094
   radius=3.1441
   shapeName=Dyn_hair_Front_R_02_capsule
   center=<2.7773 0.0418 1.3706>
   rotation=<rot -89.3625 -179.9959 84.0979>
  }]
 Dyn_hair_Front_R_03 [{
   shapeType=Capsule
   length=4.5812
   radius=2.087
   shapeName=Dyn_hair_Front_R_03_capsule
   center=<2.502 0 0.7448>
   rotation=<rot -86.2601 0.0002 -0.0003>
  }]
 Dyn_hair_Front_R_04 [{
   shapeType=Capsule
   length=5.5248
   radius=1.9384
   shapeName=Dyn_hair_Front_R_04_capsule
   center=<3.8459 -0.0955 0.705>
   rotation=<rot -87.7251 179.9997 178.1915>
  }]
 clavicle_l [{
   shapeType=Capsule
   length=4.6779
   radius=6.1717
   shapeName=clavicle_l_capsule
   center=<6.3931 -1.0251 -2.2485>
   rotation=<rot 86.9643 -0.0014 -0.0017>
  }]
 clavicle_r [{
   shapeType=Capsule
   length=4.6779
   radius=6.1717
   shapeName=clavicle_l_capsule
   center=<-6.3931 1.0251 2.2485>
   rotation=<rot 86.9643 -0.0014 179.9983>
  }]
 spine_03 [{shapeType=Capsule, length=4.8535, radius=9.5365, shapeName=spine_03_capsule, center=<-2.0838 2.2176 0>}]
 spine_04 [{shapeType=Capsule, length=5.2699, radius=9.8801, shapeName=spine_04_capsule, center=<0 2.6192 0>}]
 spine_05 [{shapeType=Capsule, length=10.2422, radius=8.8577, shapeName=spine_05_capsule, center=<0.0464 -1.6321 0>}]
 upperarm_l [{
   shapeType=Capsule
   length=24.7198
   radius=5.2802
   shapeName=upperarm_l_capsule
   center=<12.6605 -0.1968 -0.0207>
   rotation=<rot -89.9062 179.9462 -179.9461>
  }]
 upperarm_r [{
   shapeType=Capsule
   length=24.7198
   radius=5.2802
   shapeName=upperarm_l_capsule
   center=<-12.6605 0.1968 0.0207>
   rotation=<rot -89.9062 179.9462 0.0539>
  }]
}
constraints [
 {
  bone1Name=Dyn_hair_Front_L_02
  bone2Name=Dyn_hair_Front_L_01
  swing1LimitDegrees=10
  swing2LimitDegrees=10
  twistLimitDegrees=45
  twistMotion=Locked
 }
 {
  bone1Name=Dyn_hair_Front_L_03
  bone2Name=Dyn_hair_Front_L_02
  swing1LimitDegrees=10
  swing2LimitDegrees=25
  twistLimitDegrees=45
  twistMotion=Locked
 }
 {
  bone1Name=Dyn_hair_Front_L_04
  bone2Name=Dyn_hair_Front_L_03
  swing1LimitDegrees=10
  swing2LimitDegrees=60
  twistLimitDegrees=45
  twistMotion=Locked
 }
 {
  bone1Name=Dyn_hair_Front_R_02
  bone2Name=Dyn_hair_Front_R_01
  swing1LimitDegrees=10
  swing2LimitDegrees=20
  twistLimitDegrees=45
  twistMotion=Locked
 }
 {
  bone1Name=Dyn_hair_Front_R_03
  bone2Name=Dyn_hair_Front_R_02
  swing1LimitDegrees=10
  swing2LimitDegrees=25
  twistLimitDegrees=45
  twistMotion=Locked
 }
 {
  bone1Name=Dyn_hair_Front_R_04
  bone2Name=Dyn_hair_Front_R_03
  swing1LimitDegrees=10
  swing2LimitDegrees=60
  twistLimitDegrees=45
  twistMotion=Locked
 }
 {
  bone1Name=Dyn_hair_Front_R_01
  bone2Name=Base_hair_Front_R
  swing1LimitDegrees=5
  swing2LimitDegrees=10
  twistLimitDegrees=45
  twistMotion=Locked
 }
 {
  bone1Name=Dyn_hair_Front_L_01
  bone2Name=Base_hair_Front_L
  swing1LimitDegrees=5
  swing2LimitDegrees=5
  twistLimitDegrees=45
  twistMotion=Locked
 }
]
