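# Animation Blueprint for the ElegantLilyCharm face accessory part, adding secondary motion to hair, curls, flower, leaves and earring.
# It layers the custom base mesh's pose from the clavicles onto the parent's, then chains three Control Rigs, three RigidBody simulations and per-bone AnimDynamics springs.
# The event graph derives velocity-scaled spine twist and frontend/emote flags for the rigs, and computes a gravity override from backbling-dependent presets.
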AnimBlueprint /BRCosmetics/Characters/Player/Female/Medium/Bodies/F_MED_ElegantLilyCharm/Meshes/Parts/F_MED_ElegantLilyCharm_FaceAcc_AnimBP
target_skeleton=&/F_MED_ElegantLilyCharm_FaceAcc_Skeleton
graphs [
 AnimGraph {nodes [
   n1 = Get Custom Base Mesh() -> (Custom Base Mesh:Skeletal Mesh Component Object Reference)
   n2 = Get TwistAlphaSpine05() -> (TwistAlphaSpine05:float)
   n3 = Get FrontendAndEmote() -> (FrontendAndEmote:bool)
   n4 = Control Rig!(
    Source=Layered blend per bone(
     BasePose=Copy Pose From Mesh() {bUseAttachedParent=true}
     BlendPoses_0=Copy Pose From Mesh(SourceMeshComponent=@n1) {bUseAttachedParent=true}
    ) {layerSetup [{branchFilters [{blendDepth=2, boneName=clavicle_l}, {blendDepth=2, boneName=clavicle_r}]}]}
    bAlphaBoolEnabled:bool=True
    AngleSprinting:double=@n2
    isFrontEndAndEmote:bool=@n3
   ) -> (Pose) {
    alphaInputType=Bool
    alphaBoolBlend {blendOption=Linear}
    controlRigAssetReference {blueprintRigClass=&/F_MED_ElegantLilyCharm_FaceAcc_CtrlRig.F_MED_ElegantLilyCharm_FaceAcc_CtrlRig_C}
   }
   n5 = Control Rig!(Source=@n4, bAlphaBoolEnabled=True, AngleSprinting=@n2, isInFrontendOrEmote:bool=@n3) -> (Pose) {
    alphaInputType=Bool
    alphaBoolBlend {blendOption=Linear}
    controlRigAssetReference {blueprintRigClass=&/F_MED_ElegantLilyCharm_FaceAccBackPack_CtrlRig.F_MED_ElegantLilyCharm_FaceAccBackPack_CtrlRig_C}
   }
   n6 = Control Rig!(Source=@n5, Alpha:float, isFrontend:bool=Get Frontend() -> (Frontend:bool)) -> (Pose) {
    alphaBoolBlend {blendOption=Linear}
    controlRigAssetReference {blueprintRigClass=&/F_MED_ElegantLilyCharm_FaceAcc_Frontend_CtrlRig.F_MED_ElegantLilyCharm_FaceAcc_Frontend_CtrlRig_C}
   }
   n7 = RigidBody!(ComponentPose=Local To Component(LocalPose=@n6), bAlphaBoolEnabled:bool, OverrideWorldGravity:Vector="0,0,0", ExternalForce) -> (Pose) {
    alphaInputType=Bool
    bOverrideWorldGravity=true
    overridePhysicsAsset=&/F_MED_ElegantLilyCharm_FaceAcc_Physics
    simulationSpace=BaseBoneSpace
    alphaBoolBlend {blendOption=Linear}
    baseBoneRef {boneName=spine_05}
    componentLinearAccScale=<0.8 0.8 0.8>
    componentLinearVelScale=<0.8 0.8 0.8>
    simSpaceSettings {maxAngularAcceleration=30, maxAngularVelocity=3, maxLinearAcceleration=10, maxLinearVelocity=100, velocityScaleZ=0.2, worldAlpha=0.8}
   }
   n8 = RigidBody(ComponentPose=@n7, OverrideWorldGravity="0,0,0") {
    alphaInputType=Bool
    bClampLinearTranslationLimitToRefPose=true
    bOverrideWorldGravity=true
    overridePhysicsAsset=&/F_MED_ElegantLilyCharm_FaceAcc_BackHair_Physics
    simulationSpace=BaseBoneSpace
    alphaBoolBlend {blendOption=Linear}
    baseBoneRef {boneName=spine_05}
    componentLinearAccScale=<0.6 0.8 0.6>
    componentLinearVelScale=<0.6 0.8 0.6>
    simSpaceSettings {maxAngularAcceleration=30, maxAngularVelocity=3, maxLinearAcceleration=100, maxLinearVelocity=1, velocityScaleZ=0.2, worldAlpha=0.8}
   }
   n9 = RigidBody(ComponentPose=@n8, OverrideWorldGravity="0,0,0") {
    alphaInputType=Bool
    bOverrideWorldGravity=true
    overridePhysicsAsset=&/F_MED_ElegantLilyCharm_FaceAcc_BackHairBack_Physics
    simulationSpace=BaseBoneSpace
    alphaBoolBlend {blendOption=Linear}
    baseBoneRef {boneName=spine_05}
    componentLinearAccScale=<0.6 0.6 0.6>
    componentLinearVelScale=<0.6 0.6 0.6>
    simSpaceSettings {maxAngularAcceleration=30, maxAngularVelocity=3, maxLinearAcceleration=10, maxLinearVelocity=100, velocityScaleZ=0.1, worldAlpha=0.8}
   }
   n10 = Anim Dynamics - Bone: Dyn_hair_back_L_06!(ComponentPose=@n9, bAlphaBoolEnabled:bool=True, GravityOverride:Vector="0,0,0", ExternalForce="0,0,0") -> (Pose) {
    alphaInputType=Bool
    angularDampingOverride=0.75
    angularSpringConstant=250
    bAngularSpring=true
    bGravityOverrideInSimSpace=true
    bOverrideAngularDamping=true
    bUseGravityOverride=true
    simulationSpace=BoneRelative
    alphaBoolBlend {blendOption=Linear}
    boundBone {boneName=Dyn_hair_back_L_06}
    componentLinearAccScale=<2 2 2>
    componentLinearVelScale=<2 2 2>
    physicsBodyDefinitions [{
      boundBone {boneName=Dyn_hair_back_L_06}
      boxExtents=<10 10 10>
      constraintSetup {angularTargetAxis=AxisY, angularLimitsMax=<3 5 0>, angularLimitsMin=<-3 -5 -15>, angularTarget=<0 -1 0>}
      localJointOffset=<0.1 0 0>
     }]
    relativeSpaceBone {boneName=Dyn_hair_back_L_02}
   }
   n11 = Anim Dynamics - Bone: Dyn_hair_back_L_11!(ComponentPose=@n10, Alpha=1, GravityOverride:Vector="0,0,0", ExternalForce="0,0,0") -> (Pose) {
    angularDampingOverride=0.75
    angularSpringConstant=250
    bAngularSpring=true
    bGravityOverrideInSimSpace=true
    bOverrideAngularDamping=true
    bUseGravityOverride=true
    simulationSpace=BoneRelative
    alphaBoolBlend {blendOption=Linear}
    boundBone {boneName=Dyn_hair_back_L_11}
    componentLinearAccScale=<2 2 2>
    componentLinearVelScale=<2 2 2>
    physicsBodyDefinitions [{
      boundBone {boneName=Dyn_hair_back_L_11}
      boxExtents=<10 10 10>
      constraintSetup {angularTargetAxis=AxisY, angularLimitsMax=<0 5 5>, angularLimitsMin=<0 -5 -5>, angularTarget=<0 1 0>}
      localJointOffset=<0.1 0 0>
     }]
    relativeSpaceBone {boneName=Dyn_hair_back_L_04}
   }
   n12 = Anim Dynamics - Bone: Dyn_hair_back_L_10!(ComponentPose=@n11, Alpha=1, GravityOverride:Vector="0,0,0", ExternalForce="0,0,0") -> (Pose) {
    angularDampingOverride=0.75
    angularSpringConstant=250
    bAngularSpring=true
    bGravityOverrideInSimSpace=true
    bOverrideAngularDamping=true
    bUseGravityOverride=true
    simulationSpace=BoneRelative
    alphaBoolBlend {blendOption=Linear}
    boundBone {boneName=Dyn_hair_back_L_10}
    componentLinearAccScale=<2 2 2>
    componentLinearVelScale=<2 2 2>
    physicsBodyDefinitions [{
      boundBone {boneName=Dyn_hair_back_L_10}
      boxExtents=<10 10 10>
      constraintSetup {angularTargetAxis=AxisY, angularLimitsMax=<0 5 10>, angularLimitsMin=<0 -5 -10>, angularTarget=<0 1 0>}
      localJointOffset=<0.1 0 0>
     }]
    relativeSpaceBone {boneName=Dyn_hair_back_L_04}
   }
   n13 = Anim Dynamics - Bone: Dyn_hair_back_L_09!(ComponentPose=@n12, Alpha=1, GravityOverride:Vector="0,0,0", ExternalForce="0,0,0") -> (Pose) {
    angularDampingOverride=0.75
    angularSpringConstant=250
    bAngularSpring=true
    bGravityOverrideInSimSpace=true
    bOverrideAngularDamping=true
    bUseGravityOverride=true
    simulationSpace=BoneRelative
    alphaBoolBlend {blendOption=Linear}
    boundBone {boneName=Dyn_hair_back_L_09}
    componentLinearAccScale=<2 2 2>
    componentLinearVelScale=<2 2 2>
    physicsBodyDefinitions [{
      boundBone {boneName=Dyn_hair_back_L_09}
      boxExtents=<10 10 10>
      constraintSetup {angularTargetAxis=AxisY, angularLimitsMax=<0 5 10>, angularLimitsMin=<0 -5 -10>, angularTarget=<0 1 0>}
      localJointOffset=<0.1 0 0>
     }]
    relativeSpaceBone {boneName=Dyn_hair_back_L_04}
   }
   n14 = Anim Dynamics - Bone: Dyn_hair_back_L_13!(ComponentPose=@n13, Alpha=1, GravityOverride:Vector="0,0,0", ExternalForce="0,0,0") -> (Pose) {
    angularDampingOverride=0.75
    angularSpringConstant=250
    bAngularSpring=true
    bGravityOverrideInSimSpace=true
    bOverrideAngularDamping=true
    bUseGravityOverride=true
    simulationSpace=BoneRelative
    alphaBoolBlend {blendOption=Linear}
    boundBone {boneName=Dyn_hair_back_L_13}
    componentLinearAccScale=<2 2 2>
    componentLinearVelScale=<2 2 2>
    physicsBodyDefinitions [{
      boundBone {boneName=Dyn_hair_back_L_13}
      boxExtents=<10 10 10>
      constraintSetup {angularTargetAxis=AxisY, angularLimitsMax=<0 5 10>, angularLimitsMin=<0 -5 -10>, angularTarget=<0 1 0>}
      localJointOffset=<0.1 0 0>
     }]
    relativeSpaceBone {boneName=Dyn_hair_back_L_04}
   }
   n15 = Anim Dynamics - Bone: Dyn_hair_back_L_12!(ComponentPose=@n14, Alpha=1, GravityOverride:Vector="0,0,0", ExternalForce="0,0,0") -> (Pose) {
    angularDampingOverride=0.75
    angularSpringConstant=250
    bAngularSpring=true
    bGravityOverrideInSimSpace=true
    bOverrideAngularDamping=true
    bUseGravityOverride=true
    simulationSpace=BoneRelative
    alphaBoolBlend {blendOption=Linear}
    boundBone {boneName=Dyn_hair_back_L_12}
    componentLinearAccScale=<2 2 2>
    componentLinearVelScale=<2 2 2>
    physicsBodyDefinitions [{
      collisionType=CustomSphere
      sphereCollisionRadius=2
      boundBone {boneName=Dyn_hair_back_L_12}
      boxExtents=<10 10 10>
      constraintSetup {angularTargetAxis=AxisY, angularLimitsMax=<0 5 10>, angularLimitsMin=<0 -5 -10>, angularTarget=<0 1 0>}
      localJointOffset=<0.1 0 0>
     }]
    planarLimits [{drivingBone {boneName=spine_03}, planeTransform {location=<0 -7 0>, rotation=<rot 0 0 -90>}}]
    relativeSpaceBone {boneName=Dyn_hair_back_L_04}
   }
   n16 = Anim Dynamics - Bone: Dyn_hair_back_L_07!(ComponentPose=@n15, Alpha=1, GravityOverride:Vector="0,0,0", ExternalForce="0,0,0") -> (Pose) {
    angularDampingOverride=0.75
    angularSpringConstant=250
    bAngularSpring=true
    bGravityOverrideInSimSpace=true
    bOverrideAngularDamping=true
    bUseGravityOverride=true
    simulationSpace=BoneRelative
    alphaBoolBlend {blendOption=Linear}
    boundBone {boneName=Dyn_hair_back_L_07}
    componentLinearAccScale=<2 2 2>
    componentLinearVelScale=<2 2 2>
    physicsBodyDefinitions [{
      boundBone {boneName=Dyn_hair_back_L_07}
      boxExtents=<10 10 10>
      constraintSetup {angularTargetAxis=AxisY, angularLimitsMax=<0 5 10>, angularLimitsMin=<0 -5 -10>, angularTarget=<0 1 0>}
      localJointOffset=<0.1 0 0>
     }]
    relativeSpaceBone {boneName=Dyn_hair_back_L_04}
   }
   n17 = Anim Dynamics - Bone: Dyn_hair_back_L_08!(ComponentPose=@n16, Alpha=1, GravityOverride:Vector="0,0,0", ExternalForce="0,0,0") -> (Pose) {
    angularDampingOverride=0.75
    angularSpringConstant=250
    bAngularSpring=true
    bGravityOverrideInSimSpace=true
    bOverrideAngularDamping=true
    bUseGravityOverride=true
    simulationSpace=BoneRelative
    alphaBoolBlend {blendOption=Linear}
    boundBone {boneName=Dyn_hair_back_L_08}
    componentLinearAccScale=<2 2 2>
    componentLinearVelScale=<2 2 2>
    physicsBodyDefinitions [{
      boundBone {boneName=Dyn_hair_back_L_08}
      boxExtents=<10 10 10>
      constraintSetup {angularTargetAxis=AxisY, angularLimitsMax=<0 5 10>, angularLimitsMin=<0 -5 -10>, angularTarget=<0 1 0>}
      localJointOffset=<0.1 0 0>
     }]
    relativeSpaceBone {boneName=Dyn_hair_back_L_04}
   }
   n18 = Anim Dynamics - Bone: Dyn_hair_back_L_14!(ComponentPose=@n17, Alpha=1, GravityOverride:Vector="0,0,0", ExternalForce="0,0,0") -> (Pose) {
    angularDampingOverride=0.75
    angularSpringConstant=250
    bAngularSpring=true
    bGravityOverrideInSimSpace=true
    bOverrideAngularDamping=true
    bUseGravityOverride=true
    simulationSpace=BoneRelative
    alphaBoolBlend {blendOption=Linear}
    boundBone {boneName=Dyn_hair_back_L_14}
    componentLinearAccScale=<2 2 2>
    componentLinearVelScale=<2 2 2>
    physicsBodyDefinitions [{
      collisionType=CustomSphere
      sphereCollisionRadius=2
      boundBone {boneName=Dyn_hair_back_L_14}
      boxExtents=<10 10 10>
      constraintSetup {angularTargetAxis=AxisY, angularLimitsMax=<0 5 10>, angularLimitsMin=<0 -5 -10>, angularTarget=<0 1 0>}
      localJointOffset=<0.1 0 0>
     }]
    planarLimits [{drivingBone {boneName=spine_03}, planeTransform {location=<0 -7 0>, rotation=<rot 0 0 -90>}}]
    relativeSpaceBone {boneName=Dyn_hair_back_L_12}
   }
   n19 = Anim Dynamics - Bone: Dyn_hair_Front_L_05!(ComponentPose=@n18, Alpha=1, GravityOverride:Vector="0,0,0", ExternalForce="0,0,0") -> (Pose) {
    angularDampingOverride=0.75
    angularSpringConstant=150
    bAngularSpring=true
    bOverrideAngularDamping=true
    bUseGravityOverride=true
    simulationSpace=BoneRelative
    alphaBoolBlend {blendOption=Linear}
    boundBone {boneName=Dyn_hair_Front_L_05}
    componentAppliedLinearAccClamp=<10000 10000 10000>
    componentLinearAccScale=<2 2 2>
    componentLinearVelScale=<2 2 2>
    physicsBodyDefinitions [{
      boundBone {boneName=Dyn_hair_Front_L_05}
      boxExtents=<10 10 10>
      constraintSetup {angularLimitsMax=<0 0 15>, angularLimitsMin=<0 0 -15>, angularTarget=<1 0 0>}
      localJointOffset=<0.1 0 0>
     }]
    relativeSpaceBone {boneName=Dyn_hair_Front_L_03}
   }
   n20 = Anim Dynamics - Bone: Dyn_hair_Front_R_05!(ComponentPose=@n19, Alpha=1, GravityOverride:Vector="0,0,0", ExternalForce="0,0,0") -> (Pose) {
    angularDampingOverride=0.75
    angularSpringConstant=150
    bAngularSpring=true
    bOverrideAngularDamping=true
    bUseGravityOverride=true
    simulationSpace=BoneRelative
    alphaBoolBlend {blendOption=Linear}
    boundBone {boneName=Dyn_hair_Front_R_05}
    componentAppliedLinearAccClamp=<10000 10000 10000>
    componentLinearAccScale=<2 2 2>
    componentLinearVelScale=<2 2 2>
    physicsBodyDefinitions [{
      boundBone {boneName=Dyn_hair_Front_R_05}
      boxExtents=<10 10 10>
      constraintSetup {angularLimitsMax=<0 0 15>, angularLimitsMin=<0 0 -15>, angularTarget=<1 0 0>}
      localJointOffset=<0.1 0 0>
     }]
    relativeSpaceBone {boneName=Dyn_hair_Front_R_03}
   }
   n21 = Anim Dynamics - Bone: Dyn_curl_R_01!(ComponentPose=@n20, Alpha=1, GravityOverride:Vector="0,0,0", ExternalForce="0,0,0") -> (Pose) {
    angularDampingOverride=0.75
    angularSpringConstant=100
    bAngularSpring=true
    bOverrideAngularDamping=true
    bUseGravityOverride=true
    simulationSpace=BoneRelative
    alphaBoolBlend {blendOption=Linear}
    boundBone {boneName=Dyn_curl_R_01}
    componentAppliedLinearAccClamp=<10000 10000 10000>
    componentLinearAccScale=<2 2 2>
    componentLinearVelScale=<2 2 2>
    physicsBodyDefinitions [{
      boundBone {boneName=Dyn_curl_R_01}
      boxExtents=<10 10 10>
      constraintSetup {angularLimitsMax=<0 0 3>, angularLimitsMin=<0 0 -3>, angularTarget=<1 0 0>}
      localJointOffset=<0.1 0 0>
     }]
    relativeSpaceBone {boneName=head}
   }
   n22 = Anim Dynamics - Bone: Dyn_curl_R_02!(ComponentPose=@n21, Alpha=1, GravityOverride:Vector="0,0,0", ExternalForce="0,0,0") -> (Pose) {
    angularDampingOverride=0.75
    angularSpringConstant=100
    bAngularSpring=true
    bOverrideAngularDamping=true
    bUseGravityOverride=true
    simulationSpace=BoneRelative
    alphaBoolBlend {blendOption=Linear}
    boundBone {boneName=Dyn_curl_R_02}
    componentAppliedLinearAccClamp=<10000 10000 10000>
    componentLinearAccScale=<3 3 3>
    componentLinearVelScale=<3 3 3>
    physicsBodyDefinitions [{
      boundBone {boneName=Dyn_curl_R_02}
      boxExtents=<10 10 10>
      constraintSetup {angularLimitsMax=<0 5 5>, angularLimitsMin=<0 0 -5>, angularTarget=<1 0 0>}
      localJointOffset=<0.1 0 0>
     }]
    relativeSpaceBone {boneName=Dyn_curl_R_01}
   }
   n23 = Anim Dynamics - Bone: Dyn_curl_R_03!(ComponentPose=@n22, Alpha=1, GravityOverride:Vector="0,0,0", ExternalForce="0,0,0") -> (Pose) {
    angularDampingOverride=0.75
    angularSpringConstant=100
    bAngularSpring=true
    bOverrideAngularDamping=true
    bUseGravityOverride=true
    simulationSpace=BoneRelative
    alphaBoolBlend {blendOption=Linear}
    boundBone {boneName=Dyn_curl_R_03}
    componentAppliedLinearAccClamp=<10000 10000 10000>
    componentLinearAccScale=<4 4 4>
    componentLinearVelScale=<4 4 4>
    physicsBodyDefinitions [{
      boundBone {boneName=Dyn_curl_R_03}
      boxExtents=<10 10 10>
      constraintSetup {angularLimitsMax=<0 5 5>, angularLimitsMin=<0 0 -5>, angularTarget=<1 0 0>}
      localJointOffset=<0.1 0 0>
     }]
    relativeSpaceBone {boneName=Dyn_curl_R_02}
   }
   n24 = Anim Dynamics - Bone: Dyn_curl_L_01!(ComponentPose=@n23, Alpha=1, GravityOverride:Vector="0,0,0", ExternalForce="0,0,0") -> (Pose) {
    angularDampingOverride=0.75
    angularSpringConstant=100
    bAngularSpring=true
    bOverrideAngularDamping=true
    bUseGravityOverride=true
    simulationSpace=BoneRelative
    alphaBoolBlend {blendOption=Linear}
    boundBone {boneName=Dyn_curl_L_01}
    componentAppliedLinearAccClamp=<10000 10000 10000>
    componentLinearAccScale=<2 2 2>
    componentLinearVelScale=<2 2 2>
    physicsBodyDefinitions [{
      boundBone {boneName=Dyn_curl_L_01}
      boxExtents=<10 10 10>
      constraintSetup {angularLimitsMax=<0 3 5>, angularLimitsMin=<0 -3 -5>, angularTarget=<1 0 0>}
      localJointOffset=<0.1 0 0>
     }]
    relativeSpaceBone {boneName=head}
   }
   n25 = Anim Dynamics - Bone: Dyn_curl_L_02!(ComponentPose=@n24, Alpha=1, GravityOverride:Vector="0,0,0", ExternalForce="0,0,0") -> (Pose) {
    angularDampingOverride=0.75
    angularSpringConstant=100
    bAngularSpring=true
    bOverrideAngularDamping=true
    bUseGravityOverride=true
    simulationSpace=BoneRelative
    alphaBoolBlend {blendOption=Linear}
    boundBone {boneName=Dyn_curl_L_02}
    componentAppliedLinearAccClamp=<10000 10000 10000>
    componentLinearAccScale=<3 3 3>
    componentLinearVelScale=<3 3 3>
    physicsBodyDefinitions [{
      boundBone {boneName=Dyn_curl_L_02}
      boxExtents=<10 10 10>
      constraintSetup {angularLimitsMax=<0 3 5>, angularLimitsMin=<0 -3 -5>, angularTarget=<1 0 0>}
      localJointOffset=<0.1 0 0>
     }]
    relativeSpaceBone {boneName=Dyn_curl_L_01}
   }
   n26 = Anim Dynamics - Bone: flower_low!(ComponentPose=@n25, Alpha=1, GravityOverride:Vector="0,0,0", ExternalForce="0,0,0") -> (Pose) {
    angularDampingOverride=0.8
    angularSpringConstant=50
    bAngularSpring=true
    bOverrideAngularDamping=true
    bUseGravityOverride=true
    simulationSpace=BoneRelative
    alphaBoolBlend {blendOption=Linear}
    boundBone {boneName=flower_low}
    componentAppliedLinearAccClamp=<10000 10000 10000>
    componentLinearAccScale=<2 2 2>
    componentLinearVelScale=<2 2 2>
    physicsBodyDefinitions [{
      boundBone {boneName=flower_low}
      boxExtents=<10 10 10>
      constraintSetup {angularLimitsMax=<0 3 3>, angularLimitsMin=<0 -3 -3>, angularTarget=<-1 0 0>}
      localJointOffset=<0.1 0 0>
     }]
    relativeSpaceBone {boneName=flower_main}
   }
   n27 = Anim Dynamics - Bone: flower_high!(ComponentPose=@n26, Alpha=1, GravityOverride:Vector="0,0,0", ExternalForce="0,0,0") -> (Pose) {
    angularDampingOverride=0.8
    angularSpringConstant=50
    bAngularSpring=true
    bOverrideAngularDamping=true
    bUseGravityOverride=true
    simulationSpace=BoneRelative
    alphaBoolBlend {blendOption=Linear}
    boundBone {boneName=flower_high}
    componentAppliedLinearAccClamp=<10000 10000 10000>
    componentLinearAccScale=<2 2 2>
    componentLinearVelScale=<2 2 2>
    physicsBodyDefinitions [{
      boundBone {boneName=flower_high}
      boxExtents=<10 10 10>
      constraintSetup {angularLimitsMax=<0 3 3>, angularLimitsMin=<0 -3 -3>, angularTarget=<-1 0 0>}
      localJointOffset=<0.1 0 0>
     }]
    relativeSpaceBone {boneName=flower_low}
   }
   n28 = Anim Dynamics - Bone: leaves_03!(ComponentPose=@n27, Alpha=1, GravityOverride:Vector="0,0,0", ExternalForce="0,0,0") -> (Pose) {
    angularDampingOverride=0.8
    angularSpringConstant=50
    bAngularSpring=true
    bOverrideAngularDamping=true
    bUseGravityOverride=true
    simulationSpace=BoneRelative
    alphaBoolBlend {blendOption=Linear}
    boundBone {boneName=leaves_03}
    componentAppliedLinearAccClamp=<10000 10000 10000>
    componentLinearAccScale=<2 2 2>
    componentLinearVelScale=<2 2 2>
    physicsBodyDefinitions [{
      boundBone {boneName=leaves_03}
      boxExtents=<10 10 10>
      constraintSetup {angularLimitsMax=<0 2 5>, angularLimitsMin=<0 -2 -5>, angularTarget=<-1 0 0>}
      localJointOffset=<0.1 0 0>
     }]
    relativeSpaceBone {boneName=flower_low}
   }
   n29 = Anim Dynamics - Bone: leaves_01!(ComponentPose=@n28, Alpha=1, GravityOverride:Vector="0,0,0", ExternalForce="0,0,0") -> (Pose) {
    angularDampingOverride=0.8
    angularSpringConstant=50
    bAngularSpring=true
    bOverrideAngularDamping=true
    bUseGravityOverride=true
    simulationSpace=BoneRelative
    alphaBoolBlend {blendOption=Linear}
    boundBone {boneName=leaves_01}
    componentAppliedLinearAccClamp=<10000 10000 10000>
    componentLinearAccScale=<2 2 2>
    componentLinearVelScale=<2 2 2>
    physicsBodyDefinitions [{
      boundBone {boneName=leaves_01}
      boxExtents=<10 10 10>
      constraintSetup {angularLimitsMax=<0 2 5>, angularLimitsMin=<0 -2 -5>, angularTarget=<-1 0 0>}
      localJointOffset=<0.1 0 0>
     }]
    relativeSpaceBone {boneName=flower_low}
   }
   n30 = Anim Dynamics - Bone: leaves_02!(ComponentPose=@n29, Alpha=1, GravityOverride:Vector="0,0,0", ExternalForce="0,0,0") -> (Pose) {
    angularDampingOverride=0.8
    angularSpringConstant=50
    bAngularSpring=true
    bOverrideAngularDamping=true
    bUseGravityOverride=true
    simulationSpace=BoneRelative
    alphaBoolBlend {blendOption=Linear}
    boundBone {boneName=leaves_02}
    componentAppliedLinearAccClamp=<10000 10000 10000>
    componentLinearAccScale=<2 2 2>
    componentLinearVelScale=<2 2 2>
    physicsBodyDefinitions [{
      boundBone {boneName=leaves_02}
      boxExtents=<10 10 10>
      constraintSetup {angularLimitsMax=<0 2 5>, angularLimitsMin=<0 -2 -5>, angularTarget=<-1 0 0>}
      localJointOffset=<0.1 0 0>
     }]
    relativeSpaceBone {boneName=flower_low}
   }
   n31 = Anim Dynamics - Bone: dyn_earRing!(ComponentPose=@n30, Alpha=1, GravityOverride:Vector="0,0,0", ExternalForce="0,0,0") -> (Pose) {
    angularDampingOverride=0.8
    angularSpringConstant=10
    bAngularSpring=true
    bOverrideAngularDamping=true
    bUseGravityOverride=true
    simulationSpace=BoneRelative
    alphaBoolBlend {blendOption=Linear}
    boundBone {boneName=dyn_earRing}
    componentLinearAccScale=<2 2 2>
    componentLinearVelScale=<2 2 2>
    physicsBodyDefinitions [{
      boundBone {boneName=dyn_earRing}
      boxExtents=<10 10 10>
      constraintSetup {angularTargetAxis=AxisY, angularLimitsMax=<5 3 3>, angularLimitsMin=<-10 -5 -3>, angularTarget=<0 1 0>}
      localJointOffset=<-1 0 0>
     }]
    relativeSpaceBone {boneName=head}
   }
   Output Pose(Result=Component To Local(ComponentPose=@n31)) {layerGroup=None}
   Get bIsInFrontEnd()
  ]}
 EventGraph {nodes [
   n1 = Event BlueprintUpdateAnimation()
   n2 = BreakVector(InVec=GetVelocity!(self:Actor Object Reference=TryGetPawnOwner()) -> (ReturnValue:Vector))
   n3 = float + float(
    A=MapRangeClamped(Value="Absolute (Float)"(A=@n2), InRangeA=545, InRangeB=750)
    B=MapRangeClamped(Value="Absolute (Float)"(A=@n2.Y), InRangeA=545, InRangeB=750)
   )
   > Set VelocityXMovement(VelocityXMovement:double=@n3) -> (Output_Get:double)
   n4 = GetAlphaFromReferenceTransformDelta(
    SkeletalMeshComponent=Get Custom Base Mesh()
    BoneName=spine_05
    BaseName=spine_01
    DeltaComponentType=EulerX
    InputRange=(X=-30,Y=30)
    OutputRange=(X=20,Y=-20)
    TwistAxis="1,0,0"
   )
   > Set TwistAlphaSpine05(TwistAlphaSpine05:float=FInterpTo(
     Current=Get TwistAlphaSpine05()
     Target=float * float(A=@n3, B=@n4) {is=title K2Node_PromotableOperator}
     DeltaTime=@n1.DeltaTimeX
     InterpSpeed=10
    )) -> (Output_Get:float)
   > n5 = Gravity Override State Function DEPRECATED(
    CharacterPartAnimInstance=Self-Reference()
    GameTimeInSeconds=GetGameTimeInSeconds()
    GravityOverrideParams=Select(
     Option 0:Gravity Override Params Struct Structure=Get Gravity Override Params_WithoutPack() -> (Gravity Override Params_WithoutPack:Gravity Override Params Struct Structure)
     Option 1:Gravity Override Params Struct Structure=Get Gravity Override Params_WithPack() -> (Gravity Override Params_WithPack:Gravity Override Params Struct Structure)
     Index=Get Is Wearing Backbling()
    ) -> (ReturnValue:Gravity Override Params Struct Structure)
   )
   > Set Gravity Override(Gravity Override=@n5)
   n6 = Get bIsInFrontEnd()
   n7 = Get bIsPlayingEmote()
   > Set NotFrontEndOrErmote(NotFrontEndOrErmote:bool=OR Boolean(A=NOT Boolean(A=@n6), B=@n7)) -> (Output_Get:bool)
   > Set Frontend(Frontend:bool=AND Boolean(A=NOT Boolean(A=@n7), B=@n6)) -> (Output_Get:bool)
   > Set FrontendAndEmote(FrontendAndEmote:bool=AND Boolean(A=@n6, B=@n7)) -> (Output_Get:bool)
   n8 = Event OnCharacterCustomizationLoaded()
   Set Custom Base Mesh(execute=[@n8.then, @n10.then], Custom Base Mesh:Skeletal Mesh Component Object Reference=@n8.CustomBaseMesh) -> (Output_Get:Skeletal Mesh Component Object Reference)
   > Cast To FortPlayerPawn(Object=TryGetFortPawn())
   n9 = CheckForBackbling()
   > Set Is Wearing Backbling(Is Wearing Backbling=@n9)
   > Set IsNot Wearing Backbling(IsNot Wearing Backbling:bool=NOT Boolean(A=@n9)) -> (Output_Get:bool)
   n10 = Event BlueprintInitializeAnimation()
   GetForwardVector!(InRot:Rotator="0, 0, 0") -> (ReturnValue:Vector)
   n11 = BreakVector(InVec="0, 0, 0")
   MakeVector(X=@n11, Y=@n11.Y, Z="Clamp (Float)"(Value=@n11.Z, Min=0, Max=300))
   Get Custom Base Mesh()
  ]}
]
variables [
 Custom Base Mesh {type=SkeletalMeshComponent}
 ExternalForce {type=Vector}
 Frontend {type=boolean}
 FrontendAndEmote {type=boolean}
 Gravity Override {type=Vector}
 Gravity Override Params_WithoutPack {
  type=GravityOverrideParamsStruct
  default {
   skydive_DivingDown {jointName=root, gravityOverride=<0 0 980>}
   skydive_FloatingDown {gravityOverride=<0 -300 -300>, windAmplitude=<200 -500 75>, windOffset=<0 -300 200>}
   skydive_GoingUp {gravityOverride=<0 -300 150>, windAmplitude=<200 -500 150>, windOffset=<0 -150 100>}
   skydive_Parachuting {gravityOverride=<0 -100 -980>}
  }
 }
 Gravity Override Params_WithPack {
  type=GravityOverrideParamsStruct
  default {
   skydive_DivingDown {jointName=root, gravityOverride=<0 -500 500>, windAmplitude=<50 300 0>}
   skydive_FloatingDown {gravityOverride=<0 0 -980>, windAmplitude=<200 -500 0>, windOffset=<0 -300 0>}
   skydive_GoingUp {gravityOverride=<0 0 -980>, windAmplitude=<200 -500 0>, windOffset=<0 -150 0>}
   skydive_Parachuting {gravityOverride=<0 -100 -980>}
   tactical_Sprint {gravityOverride=<0 0 -750>}
  }
 }
 Is Wearing Backbling {type=boolean}
 IsNot Wearing Backbling {type=boolean}
 NotFrontEndOrErmote {type=boolean}
 TwistAlphaSpine05 {type=number}
 VelocityXMovement {type=number}
]
